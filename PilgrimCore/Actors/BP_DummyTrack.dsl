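; BP_DummyTrack is an actor that renders a scene capture into a render target held in its TextureTarget variable.
; On BeginPlay it assigns that target to its SceneCaptureComponent2D.
; On EndPlay it clears the render target, and its construction script is empty.
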(event EventBeginPlay
  (Variables|/Script/Engine.SceneCaptureComponent2D|SetTextureTarget (Variables|Default|GetTextureTarget) (Variables|Default|GetSceneCaptureComponent2D)))

(event EventEndPlay
  (params (EndPlayReason :type EEndPlayReason%20Enum :default "Destroyed"))
  (Rendering|ClearRenderTarget2D (Variables|Default|GetTextureTarget)))

(fn ConstructionScript)
