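; BP_GemSmasher is the per-lane gem smasher on a Pilgrim music-game track, showing input, hit and sustain feedback.
; Presses set glow through custom primitive data and, except on guitar-peripheral tracks, play a timed pulse mesh.
; Hits and sustains write to Niagara Data Channels, with two looping timers emulating sustain spawn rates.
; It also provides kick-drum setup, selects linear or curved gem splines, and recolors when the pro-color setting changes.
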(event EventBeginPlay
  ;[1 1 1] Set smasher visuals to default values
  (Rendering|Material|SetCustomPrimitiveDataFloat (Variables|Default|GetSM_Smasher) 3 0.0) ; GlowOn
  (Rendering|Material|SetCustomPrimitiveDataFloat (Variables|Default|GetSM_Smasher_Pulse) 4 1.0) ; PulseAnim
  (Rendering|SetVisibility (Variables|Default|GetSM_Smasher_Pulse) false false)
  (bind _returnvalue (GameplayEventRouter|GetEventRouter "/Script/FortniteGame.FortPlayspace" self))
  (bind _output_get (Variables|Default|SetEventRouter _returnvalue))
  (Utilities|FlowControl|Sequence
    (:then_0
      (GameplayEventRouter|StartListeningToEvent _output_get "/Script/PilgrimCoreRuntime.PilgrimEvent_PlayerUsingPeripheralChanged"
        (:OnEventReceived
          (Utilities|FlowControl|Delay 0.2)
          (Rendering|Material|SetCustomPrimitiveDataFloat (Variables|Default|GetSM_Smasher) 3 0.0) ; GlowOn
        )))
    (:then_1
      (GameplayEventRouter|StartListeningToEvent _output_get "/Script/PilgrimSettingsRuntime.PilgrimEvent_PilgrimSettings_UseProColorsForAllPartsChanged"
        (:OnEventReceived
          (CallFunction|SetSmasherColor))))))

(event EventOnTrackIndexSet
  (Rendering|Material|SetCustomPrimitiveDataFloat (Variables|Default|GetSM_Smasher) 0 (CallFunction|GetTrackandLaneCPD)))

(event EventOnPilgrimGameSet
  (CallFunction|SetSmasherColor))

(event EventOnInputPressed
  (bind _sm_smasher_pulse (Variables|Default|GetSM_Smasher_Pulse))
  (Rendering|Material|SetCustomPrimitiveDataFloat (Variables|Default|GetSM_Smasher) 3 1.0) ; GlowOn
  (if (not (Class|PilgrimGame|IsGuitarPeripheralTrack (Variables|PilgrimGemSmasher|GetPilgrimGame)))
    (Rendering|SetVisibility _sm_smasher_pulse true false)
    (CallFunction|StartPulse _sm_smasher_pulse)
    (bind _returnvalue (Utilities|Time|SetTimerbyEvent (EventDispatchers|CreateEvent :!event OnPulseFinished) 0.35))
    (Variables|Default|SetPulsedFinishedTimer _returnvalue)))

(event EventOnInputReleased
  (Rendering|Material|SetCustomPrimitiveDataFloat (Variables|Default|GetSM_Smasher) 3 0.0) ; GlowOn
)

(event EventBeginSustainHoldEffect
  (params (bIsOverdrive :type Boolean))
  ;[1 1 1] Start Sustain effects here.  Set overdrive status for color.
  (Variables|Default|SetSustainIsOverdrive bIsOverdrive)
  (Utilities|Time|ClearandInvalidateTimerbyHandle (Variables|Default|GetSustainTimer10))
  (Utilities|Time|ClearandInvalidateTimerbyHandle (Variables|Default|GetSustainTimer15))
  (bind _returnvalue (Utilities|Time|SetTimerbyEvent (EventDispatchers|CreateEvent :!event Custom|Spawn%20Particle%20Event%2010) 0.1 true true))
  (Variables|Default|SetSustainTimer10 _returnvalue)
  (bind _returnvalue_2 (Utilities|Time|SetTimerbyEvent (EventDispatchers|CreateEvent :!event Custom|Spawn%20Particle%20Event%2015) 0.06666 true true))
  (Variables|Default|SetSustainTimer15 _returnvalue_2))

(event EventOverdriveFailedDuringSustainHold
  ;[1 1 1] Set sustain VFX color in edge case of failing while continuing sustain hold
  (Variables|Default|SetSustainIsOverdrive false))

(event Custom|Spawn%20Particle%20Event%2010
  ;[1 1 1] During Sustain, we must constantly spawn particles as data channels do not support SpawnRate.
  (bind _lanecolor (Default|GetLaneColor))
  (NiagaraDataChannel|WriteToNiagaraDataChannel(Legacy) :Channel "/PilgrimCore/Actors/NDC_PilgrimGemSustain.NDC_PilgrimGemSustain" :bVisibleToBlueprint true :bVisibleToNiagaraCPU true :bVisibleToNiagaraGPU true :Position (Transformation|GetWorldLocation (Variables|Default|GetParticlePosition)) :LaneColor _lanecolor :IsOverdriveGem (Variables|Default|GetSustainIsOverdrive) :SpawnRate 10 :Scale (Transformation|GetWorldScale (Variables|Default|GetParticlePosition)) :TrackIndex (Variables|PilgrimGemSmasher|GetTrackIndex)))

(event Custom|Spawn%20Particle%20Event%2015
  ;[1 1 1] During Sustain, we must constantly spawn particles as data channels do not support SpawnRate.
  (bind _lanecolor (Default|GetLaneColor))
  (NiagaraDataChannel|WriteToNiagaraDataChannel(Legacy) :Channel "/PilgrimCore/Actors/NDC_PilgrimGemSustain.NDC_PilgrimGemSustain" :bVisibleToBlueprint true :bVisibleToNiagaraCPU true :bVisibleToNiagaraGPU true :Position (Transformation|GetWorldLocation (Variables|Default|GetParticlePosition)) :LaneColor _lanecolor :IsOverdriveGem (Variables|Default|GetSustainIsOverdrive) :SpawnRate 15 :Scale (Transformation|GetWorldScale (Variables|Default|GetParticlePosition)) :TrackIndex (Variables|PilgrimGemSmasher|GetTrackIndex)))

(event EventEndSustainHoldEffect
  ;[1 1 1] End Sustain effects here.
  (Utilities|Time|ClearandInvalidateTimerbyHandle (Variables|Default|GetSustainTimer10))
  (Utilities|Time|ClearandInvalidateTimerbyHandle (Variables|Default|GetSustainTimer15)))

(event EventOnInputUpdated
  (params (BoundKey :type Key%20Structure)))

(event EventOnGemSlotRelativeScaleChanged
  (Transformation|SetRelativeScale3D (Variables|Default|GetSM_Smasher) (Variables|/Script/Engine.SceneComponent|GetRelativeScale3D (Variables|PilgrimGemSmasher|GetGemSlot))))

(fn GetLaneColor
  (return (select (Variables|Default|GetTrackIsUsingProColors) (+ (CallFunction|GetTrackLaneIndex) 1) 0)))

(fn GetPerfectHitProbability
  (params (TierIndex :type Integer))
  (if (== TierIndex 1)
    (return 0.0)
    (else
      (return 1.0))))

(fn GetSmasherTexture
  (params (Key :type Key%20Structure)))

(fn InitKickGemSmasher
  (bind _sm_smasher (Variables|Default|GetSM_Smasher))
  (bind _sm_smasher_pulse (Variables|Default|GetSM_Smasher_Pulse))
  (Components|StaticMesh|SetStaticMesh _sm_smasher "/PilgrimCore/GemAndTrack/Track/Meshes/SM_Smasher_Kick.SM_Smasher_Kick")
  (Rendering|Material|SetMaterial _sm_smasher 0 "/PilgrimCore/GemAndTrack/Track/Materials/MI_Smasher_Drum.MI_Smasher_Drum")
  (Components|StaticMesh|SetStaticMesh _sm_smasher_pulse "/PilgrimCore/GemAndTrack/Track/Meshes/SM_Smasher_Pulse_Kick.SM_Smasher_Pulse_Kick")
  (Rendering|Material|SetMaterial _sm_smasher_pulse 0 "/PilgrimCore/GemAndTrack/Track/Materials/MI_Smasher_Pulse_Drum.MI_Smasher_Pulse_Drum")
  (Variables|Default|SetIsKickDrumHit true))

(fn OnPulseFinished
  (Rendering|SetVisibility (Variables|Default|GetSM_Smasher_Pulse) false false)
  (Utilities|Time|ClearandInvalidateTimerbyHandle (Variables|Default|GetPulsedFinishedTimer)))

(fn PlaySmashAnim
  (params (Accuracy :type Pilgrim%20Gem%20Hit%20Accuracy%20Structure) (IsOverdrive :type Boolean))
  (bind (_accuracyms _accuracytier _tierindex) (Utilities|Struct|BreakPilgrimGemHitAccuracy Accuracy))
  (bind _particleposition (Variables|Default|GetParticlePosition))
  (bind _returnvalue (Transformation|GetWorldScale _particleposition))
  (bind _particleposition_2 (Variables|Default|GetParticlePosition))
  (bind _returnvalue_2 (Transformation|GetWorldLocation _particleposition_2))
  (bind _trackindex (Variables|PilgrimGemSmasher|GetTrackIndex))
  (bind _perfecthitprobabillity (CallFunction|GetPerfectHitProbability _tierindex))
  (bind _lanecolor (Default|GetLaneColor))
  (if (Variables|Default|GetIsKickDrumHit)
    (NiagaraDataChannel|WriteToNiagaraDataChannel(Legacy) :Channel "/PilgrimCore/Actors/NDC_PilgrimGem_KickDrum.NDC_PilgrimGem_KickDrum" :bVisibleToBlueprint true :bVisibleToNiagaraCPU true :bVisibleToNiagaraGPU true :Position _returnvalue_2 :PerfectHitProbabilty _perfecthitprobabillity :LaneColor _lanecolor :IsOverdriveGem IsOverdrive :Scale _returnvalue :TrackIndex _trackindex)
    (else
      (NiagaraDataChannel|WriteToNiagaraDataChannel(Legacy) :Channel "/PilgrimCore/Actors/NDC_PilgrimGem.NDC_PilgrimGem" :bVisibleToBlueprint true :bVisibleToNiagaraCPU true :bVisibleToNiagaraGPU true :Position _returnvalue_2 :PerfectHitProbabilty _perfecthitprobabillity :LaneColor _lanecolor :IsOverdriveGem IsOverdrive :Scale _returnvalue :TrackIndex _trackindex))))

(fn SetGemPath
  (params (NewPath :type PilgrimGemPathTypes%20Enum :default "NewEnumerator0"))
  (switch Utilities|FlowControl|Switch|SwitchonPilgrimGemPathTypes NewPath
    (:NewEnumerator3
      (Variables|PilgrimGemSmasher|SetGemSplinePath (Variables|Default|GetCurveSpline)))
    (:NewEnumerator0
      (Variables|PilgrimGemSmasher|SetGemSplinePath (Variables|Default|GetLinearSpline)))))

(fn SetGemSmasherMeshVisibility
  (params (IsVisible :type Boolean))
  (Rendering|SetVisibility (Variables|Default|GetSM_Smasher) IsVisible false))

(fn SetSmasherColor
  (bind _sustainisoverdrive (Variables|Default|GetSustainIsOverdrive))
  (bind _output_get (Variables|Default|SetTrackIsUsingProColors (Class|PilgrimGame|IsTrackUsingProColors (Variables|PilgrimGemSmasher|GetPilgrimGame))))
  (CallFunction|SetItemFlagsonStaticMeshComponent (Variables|Default|GetSM_Smasher) _output_get _sustainisoverdrive)
  (CallFunction|SetItemFlagsonStaticMeshComponent (Variables|Default|GetSM_Smasher_Pulse) _output_get _sustainisoverdrive)
  (Rendering|Material|SetCustomPrimitiveDataFloat (Variables|Default|GetSM_Smasher) 0 (CallFunction|GetTrackandLaneCPD)))

(fn ConstructionScript
  (Utilities|FlowControl|Sequence
    (:then_0
      (CallFunction|SetGemPath "NewEnumerator3"))
    (:then_1
      (bind _returnvalue (Rendering|Material|CreateDynamicMaterialInstance (Variables|Default|GetSM_Smasher) 0 0 "None" :!class /Script/Engine.PrimitiveComponent))
      (Variables|Default|SetSmasher%20Material _returnvalue)
      (bind _returnvalue_2 (Rendering|Material|CreateDynamicMaterialInstance (Variables|Default|GetSM_Smasher_Pulse) 0 0 "None" :!class /Script/Engine.PrimitiveComponent))
      (Variables|Default|SetSmasher%20Pulse%20Material _returnvalue_2))))
