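; Now bar for the Pilgrim note track, holding five child-actor gem smashers mapped by lane that animate when gems are smashed.
; SetupNumLanes positions smashers for four or five lanes, mobile expanded tracks and lefty flip, rerunning when those settings change.
; Plastic drum tracks use their own order and turn the fifth smasher into a centered kick gem.
; The track index feeds three Niagara explosion systems, and beat markers follow a curved or linear spline.
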(event EventBeginPlay
  (Default|InitSmashers))

(event EventOnTrackIndexSet
  (bind _trackindex (Variables|PilgrimNowBar|GetTrackIndex))
  (Niagara|SetNiagaraVariable(Int32) (Variables|Default|GetNS_NDC_Explosion_Gem) "EmitterTrackIndex" _trackindex)
  (Niagara|SetNiagaraVariable(Int32) (Variables|Default|GetNS_NDC_Explosion_Gem_Sustain) "EmitterTrackIndex" _trackindex)
  (Niagara|SetNiagaraVariable(Int32) (Variables|Default|GetNS_NDC_Explosion_Gem_KickDrum) "EmitterTrackIndex" _trackindex))

(event EventOnPilgrimGameSet
  (bind _pilgrimgame (Variables|PilgrimNowBar|GetPilgrimGame))
  (Utilities|IsValid _pilgrimgame
    (:Is%20Valid
      (Variables|Default|SetIsPlasticDrums (Class|PilgrimGame|IsDrumPeripheralTrack _pilgrimgame))
      (CallFunction|SetupNumLanes (Variables|Default|GetNumSmashers))
      (bind _eventrouter (Pilgrim|PilgrimWaitForEventRouter (Components|GetOwner _pilgrimgame))
        (:OnEventRouterAvailable
          (GameplayEventRouter|StartListeningToEvent _eventrouter "/Script/PilgrimSettingsRuntime.PilgrimEvent_PilgrimSettings_UseGuitarLeftyFlipChanged"
            (:then
              (GameplayEventRouter|StartListeningToEvent _eventrouter "/Script/PilgrimSettingsRuntime.PilgrimEvent_PilgrimSettings_EnableMobileExpandedTrackChanged"
                (:OnEventReceived
                  _)))
            (:OnEventReceived
              _))
          (CallFunction|SetupNumLanes (Variables|Default|GetNumSmashers)))))))

(fn InitSmashers
  (CallFunction|SetSmasher "TrackLane0" (Variables|Default|GetGemSmasher1))
  (CallFunction|SetSmasher "TrackLane1" (Variables|Default|GetGemSmasher2))
  (CallFunction|SetSmasher "TrackLane2" (Variables|Default|GetGemSmasher3))
  (CallFunction|SetSmasher "TrackLane3" (Variables|Default|GetGemSmasher4))
  (CallFunction|SetSmasher "TrackLane4" (Variables|Default|GetGemSmasher5))
  (CallFunction|MakeSmasherArray))

(fn MakeSmasherArray
  (Utilities|Array|Clear (Variables|Default|GetAllSmashers))
  (for _array_element (Utilities|Array|MakeArray (Variables|Default|GetGemSmasher1) (Variables|Default|GetGemSmasher2) (Variables|Default|GetGemSmasher3) (Variables|Default|GetGemSmasher4) (Variables|Default|GetGemSmasher5))
    (bind _asbp_gem_smasher (Utilities|Casting|CastToBP_GemSmasher (Variables|/Script/Engine.ChildActorComponent|GetChildActor _array_element)))
    (Utilities|Array|Add (Variables|Default|GetAllSmashers) _asbp_gem_smasher)))

(fn OnGameEnded)

(fn OnGemSmashed
  (params (Lane :type EPilgrimTrackLane%20Enum%20%28by%20ref%29 :default "TrackLaneNone") (Accuracy :type Pilgrim%20Gem%20Hit%20Accuracy%20Structure) (IsOverdriveGem :type Boolean))
  (Class|BPGemSmasher|PlaySmashAnim (Utilities|Map|Find (Variables|Default|GetSmasherMap) Lane) Accuracy IsOverdriveGem))

(fn SetBeatMarkerPath
  (params (Selection :type PilgrimGemPathTypes%20Enum :default "NewEnumerator3"))
  (switch Utilities|FlowControl|Switch|SwitchonPilgrimGemPathTypes Selection
    (:NewEnumerator3
      (Variables|PilgrimNowBar|SetBeatMarkerSplinePath (Variables|Default|GetCurveSpline)))
    (:NewEnumerator0
      (Variables|PilgrimNowBar|SetBeatMarkerSplinePath (Variables|Default|GetLinearSpline)))))

(fn SetSmasher
  (params (Lane :type EPilgrimTrackLane%20Enum :default "TrackLane0") (Smasher :type Child%20Actor%20Component%20Object%20Reference))
  (bind _childactor (Variables|/Script/Engine.ChildActorComponent|GetChildActor Smasher))
  (Utilities|FlowControl|Sequence
    (:then_0
      (bind _aspilgrim_gem_smasher (Utilities|Casting|CastToPilgrimGemSmasher _childactor))
      (CallFunction|SetGemSmasher Lane _aspilgrim_gem_smasher))
    (:then_1
      (bind _asbp_gem_smasher (Utilities|Casting|CastToBP_GemSmasher _childactor))
      (Utilities|Map|Add (Variables|Default|GetSmasherMap) Lane _asbp_gem_smasher))))

(fn SetupNumLanes
  (params (Numlanes :type Integer))
  (Utilities|FlowControl|Sequence
    (:then_0
      (LocalVariables|SetUseMobileExpandedTrack (Class|PilgrimCoreBPFL|IsMobileandShouldUseExpandedTrack)))
    (:then_1
      ;[0.9713 0.3431 1] Ordering the GemShmashers inorder/reverse order based off of lefty flip settings
      (if (Variables|Default|GetIsPlasticDrums)
        ;;[0.3385 0 0.05741] Pro Drums Order (Will not include lefty flip with current implementation)
        (Utilities|Array|Add (LocalVariables|GetLocalGemsSmashers) (Variables|Default|GetGemSmasher2))
        (Utilities|Array|Add (LocalVariables|GetLocalGemsSmashers) (Variables|Default|GetGemSmasher3))
        (Utilities|Array|Add (LocalVariables|GetLocalGemsSmashers) (Variables|Default|GetGemSmasher4))
        (Utilities|Array|Add (LocalVariables|GetLocalGemsSmashers) (Variables|Default|GetGemSmasher1))
        (Utilities|Array|Add (LocalVariables|GetLocalGemsSmashers) (Variables|Default|GetGemSmasher5))
        (elif (Class|PilgrimGame|IsUsingLeftyFlip (Variables|PilgrimNowBar|GetPilgrimGame))
          ;;[0.4072 0 0.07036] In Order
          (Utilities|Array|Add (LocalVariables|GetLocalGemsSmashers) (Variables|Default|GetGemSmasher5))
          (Utilities|Array|Add (LocalVariables|GetLocalGemsSmashers) (Variables|Default|GetGemSmasher4))
          (Utilities|Array|Add (LocalVariables|GetLocalGemsSmashers) (Variables|Default|GetGemSmasher3))
          (Utilities|Array|Add (LocalVariables|GetLocalGemsSmashers) (Variables|Default|GetGemSmasher2))
          (Utilities|Array|Add (LocalVariables|GetLocalGemsSmashers) (Variables|Default|GetGemSmasher1))
          (else
            ;;[0.474 0 0.08007] Reverse Order
            (Utilities|Array|Add (LocalVariables|GetLocalGemsSmashers) (Variables|Default|GetGemSmasher1))
            (Utilities|Array|Add (LocalVariables|GetLocalGemsSmashers) (Variables|Default|GetGemSmasher2))
            (Utilities|Array|Add (LocalVariables|GetLocalGemsSmashers) (Variables|Default|GetGemSmasher3))
            (Utilities|Array|Add (LocalVariables|GetLocalGemsSmashers) (Variables|Default|GetGemSmasher4))
            (Utilities|Array|Add (LocalVariables|GetLocalGemsSmashers) (Variables|Default|GetGemSmasher5))))))
    (:then_2
      (Variables|Default|SetNumSmashers Numlanes)
      (LocalVariables|SetIs5Lanes (== 5 Numlanes))
      (LocalVariables|SetSmasherLocations (select (LocalVariables|GetIs5Lanes) (select (LocalVariables|GetUseMobileExpandedTrack) (Variables|Default|GetSmasherLocations5Lanes_Expanded) (Variables|Default|GetSmasherLocations5Lanes)) (select (LocalVariables|GetUseMobileExpandedTrack) (Variables|Default|GetSmasherLocations4Lanes_Expanded) (Variables|Default|GetSmasherLocations4Lanes))))
      (if (Variables|Default|GetIsPlasticDrums)
        ;[0.15 0.04894 0.1491 0.5] Setting up Kick Gem lane
        ;; Repositioning Kick Gem to center
        (Transformation|SetRelativeLocation (Variables|Default|GetGemSmasher5) (+ (Utilities|Array|Get(acopy) (LocalVariables|GetSmasherLocations) 1) "2.500000,0.000000,0.000000") false false)
        ;; Changing mesh and material for Kick Gem
        (bind _asbp_gem_smasher (Utilities|Casting|CastToBP_GemSmasher (Variables|/Script/Engine.ChildActorComponent|GetChildActor (Variables|Default|GetGemSmasher5)))
          (:then
            (Class|BPGemSmasher|InitKickGemSmasher _asbp_gem_smasher))
          (:CastFailed
            _))
        (else
          (Rendering|SetVisibility (Variables|Default|GetGemSmasher5) (LocalVariables|GetIs5Lanes) false)))
      (if (LocalVariables|GetIs5Lanes)
        (Transformation|SetRelativeLocation (Utilities|Array|Get(acopy) (LocalVariables|GetLocalGemsSmashers) 4) (Utilities|Array|Get(acopy) (LocalVariables|GetSmasherLocations) 4) false false)
        (else
          (Utilities|Array|RemoveItem (LocalVariables|GetLocalGemsSmashers) (Variables|Default|GetGemSmasher5))))
      (Transformation|SetRelativeLocation (Utilities|Array|Get(acopy) (LocalVariables|GetLocalGemsSmashers)) (Utilities|Array|Get(acopy) (LocalVariables|GetSmasherLocations)) false false)
      (Transformation|SetRelativeLocation (Utilities|Array|Get(acopy) (LocalVariables|GetLocalGemsSmashers) 1) (Utilities|Array|Get(acopy) (LocalVariables|GetSmasherLocations) 1) false false)
      (Transformation|SetRelativeLocation (Utilities|Array|Get(acopy) (LocalVariables|GetLocalGemsSmashers) 2) (Utilities|Array|Get(acopy) (LocalVariables|GetSmasherLocations) 2) false false)
      (Transformation|SetRelativeLocation (Utilities|Array|Get(acopy) (LocalVariables|GetLocalGemsSmashers) 3) (Utilities|Array|Get(acopy) (LocalVariables|GetSmasherLocations) 3) false false))))

(fn ConstructionScript)
