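; BP_PilgrimBeatMarker is a beat marker actor placed on a track in the Pilgrim rhythm game.
; When added to the track, it calls the parent logic, then sets custom primitive data float 4 on its SM_BeatMarker mesh to zero to initialize opacity.
; When the game ends, it calls the parent and sets a GameEnded flag, a hook noted for art use.
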(event EventOnAddedToTrack
  (|CallParentFunction)
  ;[0.3333 0.6781 1] Init Opacity Value
  (Rendering|Material|SetCustomPrimitiveDataFloat (Variables|Default|GetSM_BeatMarker) 4 0.0))

(event EventOnGameEnded
  ;[1 1 1] Art: This is where beatmarkers get notified the game is ending.
  (|CallParentFunction)
  (Variables|Default|SetGameEnded true))

(fn ConstructionScript)
