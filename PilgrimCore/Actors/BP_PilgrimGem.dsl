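; BP_PilgrimGem is the note gem track item for Pilgrim, the rhythm game, covering normal, pull-off, HOPO, kick, cymbal and overdrive-deploy types.
; When added to the track, it picks mesh, materials and glow from its type, with cymbal visuals gated by a check.
; Custom primitive data drives miss, desaturation and glow alpha, while overdrive toggles glow visibility.
; Smashes finish immediately without effects, and calibration debug adds a text label.
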(event EventOnAddedToTrack
  (|CallParentFunction)
  (CallFunction|SetGem))

(event EventOnOverdriveChanged
  (params (bIsOverdrive :type Boolean))
  (|CallParentFunction bIsOverdrive)
  (CallFunction|SetGemOverdrive bIsOverdrive))

(event EventOnGemSmashed
  ;[1 1 1] Gem "Smash" effects can be hooked up here. There is also a function in GemSmasher that gets called on successful hits. Currently only GemSmasher has any effects.
  ;
  ; If you do hook up a gem smash effect, try to trigger Finish Gem Smash at the end of it.
  (|CallParentFunction)
  (CallFunction|FinishGemSmash))

(event EventOnPassedNowBar
  (params (bPassedNowBar :type Boolean))
  (bind _gemmesh (Variables|Default|GetGemMesh))
  (bind _gem_glow (Variables|Default|GetGem_Glow))
  ;[1 1 1] Gem Pass/Miss logic goes here! If you need to reset, consider in "SetGemLane"
  (|CallParentFunction bPassedNowBar)
  ;[0.3438 0.3438 0.3438] Set Gem Miss Params
  (Rendering|Material|SetCustomPrimitiveDataFloat _gemmesh 5 1.0) ; IsMiss
  (Rendering|Material|SetCustomPrimitiveDataFloat _gemmesh 6 0.25) ; Desaturation
  (Rendering|Material|SetCustomPrimitiveDataFloat _gem_glow 4 0.0) ; Alpha
  (Rendering|SetVisibility _gem_glow false false))

(event EventOnGameEnded
  ;[1 1 1] Art: Here is where the gem gets notified the game is ending. You have a few seconds to tear down.
  (|CallParentFunction))

(fn IsInputSpecificGemShapesEnabled
  (if (Utilities|IsValid (Variables|PilgrimTrackItem|GetParentGame) :!class /Script/Engine.KismetSystemLibrary)
    (return (Class|PilgrimGame|IsInputSpecificGemShapesEnabled (Variables|PilgrimTrackItem|GetParentGame)))
    (else
      (return false))))

(fn OnCalibrationDebugSet
  (params (bDebugEnabled :type Boolean))
  (if bDebugEnabled
    (Utilities|IsValid (Variables|Default|GetDebugTextComponent)
      (:Is%20Valid
        _)
      (:Is%20Not%20Valid
        (bind _returnvalue (AddComponent|NODE_AddTextRenderComponent-1 :RelativeTransform (Math|Transform|MakeTransform "0.000000,2.000000,1.000000" "30.000000,90.000000,0.000000")))
        (Variables|Default|SetDebugTextComponent _returnvalue)
        (Utilities|Array|Add (LocalVariables|GetDebugPrimitives) (Variables|Default|GetDebugTextComponent))))
    (Rendering|Components|TextRender|SetText (Variables|Default|GetDebugTextComponent) (Utilities|Text|ToText(Integer) (Utilities|Struct|BreakTrackItemInfo (CallFunction|GetItemInfo))))
    (else
      (Utilities|IsValid (Variables|Default|GetDebugTextComponent)
        (:Is%20Valid
          (Utilities|Array|Add (LocalVariables|GetDebugPrimitives) (Variables|Default|GetDebugTextComponent))
          (Components|DestroyComponent (Variables|Default|GetDebugTextComponent)))
        (:Is%20Not%20Valid
          _))))
  (return (LocalVariables|GetDebugPrimitives)))

(fn SetGem
  (bind _returnvalue (CallFunction|GetItemInfo))
  (bind (_gemtype _isoverdrivegem) (Utilities|Struct|BreakTrackItemInfo _returnvalue))
  (Default|SetGemType _gemtype)
  (CallFunction|SetInitialGemColor _gemtype)
  (if (Utilities|Enum|Equal(Enum) _gemtype "DeployOverdrive")
    (CallFunction|SetGemOverdrive true)
    (else
      (CallFunction|SetGemOverdrive _isoverdrivegem))))

(fn SetGemOverdrive
  (params (IsOverdrive :type Boolean))
  (Rendering|SetVisibility (Variables|Default|GetGem_Glow) IsOverdrive false))

(fn SetGemType
  (params (GemType :type EPilgrimGemType%20Enum :default "Normal"))
  (switch Utilities|FlowControl|Switch|SwitchonEPilgrimGemType GemType
    (:Normal
      _)
    (:PullOff
      (Components|StaticMesh|SetStaticMesh (Variables|Default|GetGemMesh) "/PilgrimCore/GemAndTrack/Gems/Meshes/SM_Gem_Pull_Off.SM_Gem_Pull_Off")
      (Transformation|SetRelativeScale3D (Variables|Default|GetGem_Glow) "0.035000,0.035000,0.030000")
      (Rendering|Material|SetMaterial (Variables|Default|GetGemMesh) 0 "/PilgrimCore/GemAndTrack/Gems/Materials/MI_Gem_PullOff.MI_Gem_PullOff")
      (Rendering|Material|SetMaterial (Variables|Default|GetGemMesh) 1 "/PilgrimCore/GemAndTrack/Gems/Materials/M_Gem_Shadow.M_Gem_Shadow")
      (Rendering|Material|SetMaterial (Variables|Default|GetGem_Glow) 0 "/PilgrimCore/GemAndTrack/Gems/Materials/M_Gem_Glow.M_Gem_Glow"))
    (:HOPO
      (Components|StaticMesh|SetStaticMesh (Variables|Default|GetGemMesh) "/PilgrimCore/GemAndTrack/Gems/Meshes/SM_Gem_HOPO.SM_Gem_HOPO")
      (Transformation|SetRelativeScale3D (Variables|Default|GetGem_Glow) "0.035000,0.035000,0.030000")
      (Rendering|Material|SetMaterial (Variables|Default|GetGemMesh) 0 "/PilgrimCore/GemAndTrack/Gems/Materials/MI_Gem_HOPO.MI_Gem_HOPO")
      (Rendering|Material|SetMaterial (Variables|Default|GetGemMesh) 1 "/PilgrimCore/GemAndTrack/Gems/Materials/M_Gem_Shadow.M_Gem_Shadow")
      (Rendering|Material|SetMaterial (Variables|Default|GetGem_Glow) 0 "/PilgrimCore/GemAndTrack/Gems/Materials/M_Gem_Glow.M_Gem_Glow"))
    (:Kick
      (Components|StaticMesh|SetStaticMesh (Variables|Default|GetGemMesh) "/PilgrimCore/GemAndTrack/Gems/Meshes/SM_Gem_KickDrum.SM_Gem_KickDrum")
      (Transformation|SetRelativeScale3D (Variables|Default|GetGem_Glow) "0, 0, 0")
      (Rendering|Material|SetMaterial (Variables|Default|GetGemMesh) 0 "/PilgrimCore/GemAndTrack/Gems/Materials/MI_Gem_Kick.MI_Gem_Kick")
      (Rendering|Material|SetMaterial (Variables|Default|GetGemMesh) 1 "/PilgrimCore/GemAndTrack/Gems/Materials/MI_Gem_Shadow_Invis.MI_Gem_Shadow_Invis")
      (Rendering|Material|SetMaterial (Variables|Default|GetGem_Glow) 0 "/PilgrimCore/GemAndTrack/Gems/Materials/M_Gem_Glow.M_Gem_Glow"))
    (:Cymbal
      (if (CallFunction|CheckifCymbalGemTypeIsAllowed)
        (Components|StaticMesh|SetStaticMesh (Variables|Default|GetGemMesh) "/PilgrimCore/GemAndTrack/Gems/Meshes/SM_Gem_Cymbal_Drum.SM_Gem_Cymbal_Drum")
        (Transformation|SetRelativeScale3D (Variables|Default|GetGem_Glow) "0.035000,0.035000,0.030000")
        (Rendering|Material|SetMaterial (Variables|Default|GetGemMesh) 0 "/PilgrimCore/GemAndTrack/Gems/Materials/MI_Gem_Cymbal.MI_Gem_Cymbal")
        (Rendering|Material|SetMaterial (Variables|Default|GetGemMesh) 1 "/PilgrimCore/GemAndTrack/Gems/Materials/M_Gem_Shadow.M_Gem_Shadow")
        (Rendering|Material|SetMaterial (Variables|Default|GetGem_Glow) 0 "/PilgrimCore/GemAndTrack/Gems/Materials/M_Gem_Glow.M_Gem_Glow")
        (else
          _)))
    (:DeployOverdrive
      (if (CallFunction|CheckifCymbalGemTypeIsAllowed)
        (Components|StaticMesh|SetStaticMesh (Variables|Default|GetGemMesh) "/PilgrimCore/GemAndTrack/Gems/Meshes/SM_Gem_Cymbal_Drum.SM_Gem_Cymbal_Drum")
        (Transformation|SetRelativeScale3D (Variables|Default|GetGem_Glow) "0.050000,0.050000,0.040000")
        (Rendering|Material|SetMaterial (Variables|Default|GetGemMesh) 0 "/PilgrimCore/GemAndTrack/Gems/Materials/MI_Gem_Cymbal.MI_Gem_Cymbal")
        (Rendering|Material|SetMaterial (Variables|Default|GetGemMesh) 1 "/PilgrimCore/GemAndTrack/Gems/Materials/M_Gem_Shadow.M_Gem_Shadow")
        (Rendering|Material|SetMaterial (Variables|Default|GetGem_Glow) 0 "/PilgrimCore/GemAndTrack/Gems/Materials/MI_Gem_Glow_DrumODDeploy.MI_Gem_Glow_DrumODDeploy")
        (else
          (Components|StaticMesh|SetStaticMesh (Variables|Default|GetGemMesh) "/PilgrimCore/GemAndTrack/Gems/Meshes/SM_Gem_Half_Pill.SM_Gem_Half_Pill")
          (Transformation|SetRelativeScale3D (Variables|Default|GetGem_Glow) "0.050000,0.050000,0.050000")
          (Rendering|Material|SetMaterial (Variables|Default|GetGemMesh) 0 "/PilgrimCore/GemAndTrack/Gems/Materials/M_Gem.M_Gem")
          (Rendering|Material|SetMaterial (Variables|Default|GetGemMesh) 1 "/PilgrimCore/GemAndTrack/Gems/Materials/M_Gem_Shadow.M_Gem_Shadow")
          (Rendering|Material|SetMaterial (Variables|Default|GetGem_Glow) 0 "/PilgrimCore/GemAndTrack/Gems/Materials/MI_Gem_Glow_DrumODDeploy.MI_Gem_Glow_DrumODDeploy")))))
  (Components|StaticMesh|SetStaticMesh (Variables|Default|GetGemMesh) "/PilgrimCore/GemAndTrack/Gems/Meshes/SM_Gem_Half_Pill.SM_Gem_Half_Pill")
  (Transformation|SetRelativeScale3D (Variables|Default|GetGem_Glow) "0.050000,0.050000,0.050000")
  (Rendering|Material|SetMaterial (Variables|Default|GetGemMesh) 0 "/PilgrimCore/GemAndTrack/Gems/Materials/M_Gem.M_Gem")
  (Rendering|Material|SetMaterial (Variables|Default|GetGemMesh) 1 "/PilgrimCore/GemAndTrack/Gems/Materials/M_Gem_Shadow.M_Gem_Shadow")
  (Rendering|Material|SetMaterial (Variables|Default|GetGem_Glow) 0 "/PilgrimCore/GemAndTrack/Gems/Materials/M_Gem_Glow.M_Gem_Glow"))

(fn SetInitialGemColor
  (params (GemType :type EPilgrimGemType%20Enum :default "Normal"))
  ;[1 1 1] CPD Defaults
  (Rendering|Material|SetCustomPrimitiveDataFloat (Variables|Default|GetGemMesh) 5 0.0) ; IsMiss
  (Rendering|Material|SetCustomPrimitiveDataFloat (Variables|Default|GetGemMesh) 6 (Utilities|Select :Index GemType :Normal 0.0 :PullOff 0.65 :HOPO 0.0 :Kick 0.0 :Cymbal 0.0 :DeployOverdrive 0.0)) ; Desaturation
  (Rendering|Material|SetCustomPrimitiveDataFloat (Variables|Default|GetGem_Glow) 4 1.0) ; Alpha (for OD glow)
  (if (Utilities|Enum|Equal(Enum) GemType "Kick")
    (Rendering|Material|SetColorParameterValueonMaterials (Variables|Default|GetGemMesh) "Gem Color" (Math|Conversions|ToLinearColor(Color) (Variables|Default|GetColor_Orange)))))

(fn ConstructionScript)
