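; BP_PilgrimTrack is the per-player note track in the Pilgrim music game, hosting the now bar, streak meter and overdrive meter.
; Gameplay event router listeners drive 4/5-lane setup, overdrive and max-streak effects, and settings-based colors, width, divider and lefty flip.
; Timelines phase the track in and out around songs and gem breaks, swapping animated and default materials.
; Gem hits forward to the now bar and write lane feedback into custom primitive data.
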(event EventBeginPlay
  (CallFunction|AddComponentstoTrackCapture)
  ;[0.2396 0.1393 0.05766] Kinda sneaky. hope to undo it. Right now this Component is set to 2* the end of the track to compensate for a .5 scale on the nowbar actor. This helps to make sure APilgrimTrack::GetTrackSpeedPerSecond returns values that make sense.
  (Variables|PilgrimTrack|SetTopOfTrack (Variables|Default|GetTopOfTrackSceneComponent))
  (bind _returnvalue (GameplayEventRouter|GetEventRouter "/Script/FortniteGame.FortPlayspace" self))
  (Variables|Default|SetEventRouter _returnvalue)
  (CallFunction|BindtoGameplayEvents)
  (CallFunction|SetupOverdriveMeter)
  (Default|SetupStreakMeter)
  (Default|ReinitializeWidgetComponents))

(event EventOnPilgrimGameSet
  (Default|SetupLeftyFlip)
  (CallFunction|SetupOverdriveVisibility)
  (Default|SetupIsPlasticDrums))

(event Custom|Event%20Track%20Phase%20In
  ;[0.2899 0.9927 1 0.577] Don't phase the track in or out if it's already in the correct state (prevent the track from popping in or out)
  (if (not (Variables|Default|GetIsTrackVisible))
    (CallFunction|SetTrackMaterial true)
    (Variables|Default|SetIsTrackVisible true)
    ;[1 1 1] Make sure track is visible
    (Rendering|SetVisibility (Variables|PilgrimTrack|GetTrackMeshComponent) true false)
    ;[1 1 1] Animate track materials
    (bind (_direction _phase_in_non_track_elements _phase_in_track) (Timeline|TimelineTrackPhaseIn PlayFromStart)
      (:Update
        (CallFunction|SetSecondaryElementsPhase _phase_in_non_track_elements) ; Drives UI widget pass through materials in BP_PilgrimTrack, gem smasher materials in BP_GemSmasher - animates slightly delayed from track entering
        (Rendering|Material|SetScalarParameterValueonMaterials (Variables|PilgrimTrack|GetTrackMeshComponent) "TrackPhaseInOut" _phase_in_track))
      (:Finished
        (CallFunction|SetTrackMaterial false)
        ;[1 1 1] Make sure track doesn't come back after phase out and switching to a new material that doesn't have animation in it to hide it
        (Rendering|SetVisibility (Variables|PilgrimTrack|GetTrackMeshComponent) (Variables|Default|GetIsTrackVisible) false)))))

(event Custom|StopPhaseInOut
  ;[1 1 1] Make sure track is visible
  ;[1 1 1] Animate track materials
  (Timeline|TimelineTrackPhaseIn Stop))

(event Custom|Event%20Track%20Phase%20Out
  ;[0.2899 0.9927 1 0.577] Don't phase the track in or out if it's already in the correct state (prevent the track from popping in or out)
  (if (Variables|Default|GetIsTrackVisible)
    (CallFunction|SetTrackMaterial true)
    (Variables|Default|SetIsTrackVisible false) ; Used to make sure track stays non visible when anim finishes
    ;[1 1 1] Animate track materials
    (Timeline|TimelineTrackPhaseIn ReverseFromEnd)))

(event EventOnGameEnded
  ;[1 1 1] Art: Here is where the track gets notified the game is ending. You have a few seconds to tear down.
  (|CallParentFunction)
  (Utilities|FlowControl|Sequence
    (:then_0
      (bind _asbp_now_bar (Utilities|Casting|CastToBP_NowBar (Variables|/Script/Engine.ChildActorComponent|GetChildActor (Variables|PilgrimTrack|GetNowBarChildActor))))
      (Class|BPNowBar|OnGameEnded _asbp_now_bar))
    (:then_1
      ;[1 1 1] Play track phase out
      (CallFunction|EventTrackPhaseOut))))

(event EventOnGemSmashed
  (params (lane :type EPilgrimTrackLane%20Enum :default "TrackLaneNone") (Accuracy :type Pilgrim%20Gem%20Hit%20Accuracy%20Structure%20%28by%20ref%29) (bIsOverdrive :type Boolean))
  ;[1 1 1] On Gem Smash Event
  (bind _asbp_now_bar (Utilities|Casting|CastToBP_NowBar (Variables|/Script/Engine.ChildActorComponent|GetChildActor (Variables|PilgrimTrack|GetNowBarChildActor))))
  (Class|BPNowBar|OnGemSmashed _asbp_now_bar lane Accuracy bIsOverdrive)
  (if (CallFunction|GetShowDetailedHitFeedback)
    (CallFunction|TriggerLaneFeedback lane Accuracy)))

(event Custom|BindToGameplayEvents
  (bind _returnvalue (Actor|GetComponentbyClass (CallFunction|GetOwningPlayer) "/PilgrimCore/Player/Components/BP_PilgrimPlayerControllerComponent.BP_PilgrimPlayerControllerComponent_C"))
  (bind (_dividerposition _dividervisibility) (Default|GetDividerParams))
  (bind _returnvalue_2 (CallFunction|GetOwningPlayer))
  (bind _eventrouter (Variables|Default|GetEventRouter))
  (Utilities|FlowControl|Sequence
    (:then_0
      ;[1 1 1] Overdrive On/Off Material
      (Rendering|Material|SetScalarParameterValueonMaterials (Variables|PilgrimTrack|GetTrackMeshComponent) "Overdrive On/Off" 0.0)
      (bind (_asynctaskproxy _payload) (GameplayEventRouter|StartListeningToEvent (Variables|Default|GetEventRouter) "/Script/PilgrimCoreRuntime.PilgrimEvent_Scoring_OverdriveActiveChanged" (Variables|/Script/Engine.Controller|GetPlayerState (CallFunction|GetOwningPlayer)))
        (:then
          (Utilities|Array|Add (Variables|Default|GetGameplayEventHandles) _asynctaskproxy))
        (:OnEventReceived
          (if (== (PlayerState|GetPlayerController (.DeployingPlayer (Utilities|Struct|BreakPilgrimEventScoringOverdriveActiveChanged _payload))) (CallFunction|GetOwningPlayer))
            (if (Utilities|Struct|BreakPilgrimEventScoringOverdriveActiveChanged _payload)
              (Components|Activation|Activate (Variables|Default|GetNS_Track_Overdrive) false)
              (Rendering|Material|SetScalarParameterValueonMaterials (Variables|PilgrimTrack|GetTrackMeshComponent) "GameTime" (Utilities|Time|GetGameTimeinSeconds))
              (bind (_direction _overdrivefade) (Timeline|Timeline_0 Play)
                (:Update
                  (Rendering|Material|SetScalarParameterValueonMaterials (Variables|PilgrimTrack|GetTrackMeshComponent) "Overdrive On/Off" _overdrivefade)))
              (else
                (Timeline|Timeline_0 Reverse)))))))
    (:then_1
      ;[1 1 1] Max Streak On/Off Material
      (bind (_asynctaskproxy_2 _payload_2) (GameplayEventRouter|StartListeningToEvent (Variables|Default|GetEventRouter) "/Script/PilgrimCoreRuntime.PilgrimEvent_Scoring_MaxStreakStateChanged" (Actor|GetOwner))
        (:then
          ;;[1 1 1] If the local player reaches or loses max streak, this logic will trigger. This does trigger for EVERY player, but the getOwner check filters it to only the local owner.
          (Utilities|Array|Add (Variables|Default|GetGameplayEventHandles) _asynctaskproxy_2))
        (:OnEventReceived
          (if (== (Actor|GetOwner) (Utilities|Struct|BreakPilgrimEventScoringMaxStreakStateChanged _payload_2))
            (if (.bIsAtMaxStreak (Utilities|Struct|BreakPilgrimEventScoringMaxStreakStateChanged _payload_2))
              (Components|Activation|Activate (Variables|Default|GetNS_Track_PeakState) false)
              (bind (_direction_2 _max_streak_fade) (Timeline|Timeline_1 Play)
                (:Update
                  (bind _output_get (Variables|Default|SetMax%20Streak%20Fade _max_streak_fade)) ; Cache value for when we set materials on track phase in/out
                  (Rendering|Material|SetScalarParameterValueonMaterials (Variables|PilgrimTrack|GetTrackMeshComponent) "Max Streak On/Off" _output_get)))
              (else
                (Timeline|Timeline_1 Reverse)))))))
    (:then_2
      ;[1 1 1] Set up lanes
      (bind (_asynctaskproxy_3 _payload_3) (GameplayEventRouter|StartListeningToEvent (Variables|Default|GetEventRouter) "/Script/PilgrimCoreRuntime.PilgrimEvent_PilgrimGame_PlayingSong")
        (:then
          (Utilities|Array|Add (Variables|Default|GetGameplayEventHandles) _asynctaskproxy_3))
        (:OnEventReceived
          (if (or (== (Class|PilgrimGame|GetLocalPlayerController (Utilities|Struct|BreakPilgrimEventPilgrimGamePlayingSong _payload_3)) (CallFunction|GetOwningPlayer)) (not (Utilities|IsValid (CallFunction|GetOwningPlayer) :!class /Script/Engine.KismetSystemLibrary)))
            (CallFunction|SetupNumLanes (Class|PilgrimGame|GetNumLanes (Utilities|Struct|BreakPilgrimEventPilgrimGamePlayingSong _payload_3)))
            ;[0.5121 1 0.869] First check to see if the player controller has an EmoteDuringGemBreak component on it. If not, allow the track to phase in, otherwise it may never fade in. If that component is on the controller, that component will add/remove the corresponding tags that will bring the track back when gem breaks end.
            (if (and (CallFunction|OwningPlayerHasEmoteDuringGemBreakComponent) (not (Class|PilgrimCoreBPFL|IsHamSandwichSessionActive)))
              ;[0.5121 1 0.869] Only phase in the track if we are not in a starting gem break
              (bind _canphasetrack (CallFunction|CanPhaseTrackatSongStart (Utilities|Struct|BreakPilgrimEventPilgrimGamePlayingSong _payload_3)))
              (if _canphasetrack
                _)
              (else
                _))
            ;[1 1 1] Play track phase in
            (CallFunction|EventTrackPhaseIn)))))
    (:then_3
      (Networking|CanExecuteCosmeticEvents
        (:True
          ;[0.5121 1 0.869] If the owning player has an EmoteDuringGemBreak component on their player controller, phase track in and out on gem break (do once is due to TagRemoveFromActor firing immediately when no tag is present).
          (if (CallFunction|OwningPlayerHasEmoteDuringGemBreakComponent)
            (Utilities|FlowControl|Sequence
              (:then_0
                ;;[1 1 1 0.5] Gem Break Start
                (bind (_asynctaskproxy_4 _payload_4) (GameplayEventRouter|StartListeningToEvent (Variables|Default|GetEventRouter) "/Script/PilgrimCoreRuntime.PilgrimEvent_GemBreak_Started" (Variables|/Script/Engine.Controller|GetPlayerState (CallFunction|GetOwningPlayer)))
                  (:then
                    ;;[0.4951 1 0.3394 0.5] Don't phase the track if there is more than one local player, unless it's the last song break
                    (Utilities|Array|Add (Variables|Default|GetGameplayEventHandles) _asynctaskproxy_4))
                  (:OnEventReceived
                    (if (not (and (Class|PilgrimCoreBPFL|IsHamSandwichSessionActive) (not (Utilities|Struct|BreakPilgrimGemBreak (Utilities|Struct|BreakPilgrimEventGemBreakBase _payload_4)))))
                      (bind _canphasetrack_2 (CallFunction|CanPhaseTrackatSongStart (Variables|PilgrimTrack|GetPilgrimGame)))
                      (if _canphasetrack_2
                        (CallFunction|EventTrackPhaseOut))))))
              (:then_1
                ;;[1 1 1 0.5] GemBreak End Pending
                (bind (_asynctaskproxy_5 _payload_5) (GameplayEventRouter|StartListeningToEvent (Variables|Default|GetEventRouter) "/Script/PilgrimCoreRuntime.PilgrimEvent_GemBreak_End_Pending" (Variables|/Script/Engine.Controller|GetPlayerState (CallFunction|GetOwningPlayer)))
                  (:then
                    (Utilities|Array|Add (Variables|Default|GetGameplayEventHandles) _asynctaskproxy_5))
                  (:OnEventReceived
                    (if (not (or (Class|PilgrimCoreBPFL|IsHamSandwichSessionActive) (Utilities|Struct|BreakPilgrimGemBreak (Utilities|Struct|BreakPilgrimEventGemBreakBase _payload_5))))
                      (CallFunction|EventTrackPhaseIn)))))
              (:then_2
                ;;[1 1 1 0.5] Gem Break End
                (bind (_asynctaskproxy_6 _payload_6) (GameplayEventRouter|StartListeningToEvent (Variables|Default|GetEventRouter) "/Script/PilgrimCoreRuntime.PilgrimEvent_GemBreak_Started" (Variables|/Script/Engine.Controller|GetPlayerState (CallFunction|GetOwningPlayer)))
                  (:then
                    (Utilities|Array|Add (Variables|Default|GetGameplayEventHandles) _asynctaskproxy_6))
                  (:OnEventReceived
                    (Variables|Default|SetIs%20Ending%20Break (.bIsEndingBreak (Utilities|Struct|BreakPilgrimGemBreak (Utilities|Struct|BreakPilgrimEventGemBreakBase _payload_6))))))))))))
    (:then_4
      ;[1 1 1] Listen for input  changes to adjust the divider, if necessary
      (bind _asynctaskproxy_7 (GameplayEventRouter|StartListeningToEvent _eventrouter "/Script/PilgrimCoreRuntime.PilgrimEvent_KeybindPresetChanged")
        (:then
          (Utilities|Array|Add (Variables|Default|GetGameplayEventHandles) _asynctaskproxy_7)
          (bind _asynctaskproxy_8 (GameplayEventRouter|StartListeningToEvent _eventrouter "/Script/PilgrimCoreRuntime.PilgrimEvent_KeybindPresetActivated")
            (:then
              (Utilities|Array|Add (Variables|Default|GetGameplayEventHandles) _asynctaskproxy_8)
              (Utilities|IsValid _returnvalue_2
                (:Is%20Valid
                  (CommonInputSubsystem|BindEventtoOnInputMethodChanged (PlayerController|LocalPlayerSubsystems|GetCommonInputSubsystem _returnvalue_2) (delegate Custom|CustomEvent_0))
                  (Utilities|IsValid _returnvalue
                    (:Is%20Valid
                      (EventDispatchers|BindEventtoOnInputSettingsChanged _returnvalue (delegate Custom|OnInputSettingsChanged)))))))
            (:OnEventReceived
              _)))
        (:OnEventReceived
          _))
      (CallFunction|SetTrackMaterial true))
    (:then_5
      (bind _asynctaskproxy_9 (GameplayEventRouter|StartListeningToEvent (Variables|Default|GetEventRouter) "/Script/PilgrimSettingsRuntime.PilgrimEvent_PilgrimSettings_EnableMobileExpandedTrackChanged")
        (:then
          (Utilities|Array|Add (Variables|Default|GetGameplayEventHandles) _asynctaskproxy_9))
        (:OnEventReceived
          (CallFunction|UpdateTrackWidth))))
    (:then_6
      (bind (_asynctaskproxy_10 _payload_7) (GameplayEventRouter|StartListeningToEvent (Variables|Default|GetEventRouter) "/Script/PilgrimSettingsRuntime.PilgrimEvent_PilgrimSettings_UseProColorsForAllPartsChanged")
        (:then
          (Utilities|Array|Add (Variables|Default|GetGameplayEventHandles) _asynctaskproxy_10))
        (:OnEventReceived
          (if (not (Class|PilgrimGame|IsGuitarPeripheralTrack (Variables|PilgrimTrack|GetPilgrimGame)))
            (Utilities|FlowControl|Sequence
              (:then_0
                ;[0.894 0.2379 1] Updating the track color scheme on setting changed
                (CallFunction|SetPlasticParams)
                (Rendering|Material|SetScalarParameterValueonMaterials (Variables|PilgrimTrack|GetTrackMeshComponent) "DividerPosition" _dividerposition)
                (Rendering|Material|SetScalarParameterValueonMaterials (Variables|PilgrimTrack|GetTrackMeshComponent) "UseDivider" _dividervisibility))
              (:then_1
                ;[1 0.4078 0.9083] Changing color scheme of UI displayed on track
                (bind _aswbp_streak_meter (Utilities|Casting|CastToWBP_StreakMeter (UserInterface|GetWidget (Variables|Default|GetStreakMeter))))
                (Class|WBPStreakMeter|UpdateStreakUIColors _aswbp_streak_meter (Utilities|Struct|BreakPilgrimEventPilgrimSettingsUseProColorsforAllPartsChanged _payload_7))
                (bind _aswbp_pilgrim_core_overdrive_meter_horizontal (Utilities|Casting|CastToWBP_PilgrimCore_OverdriveMeter_Horizontal (UserInterface|GetWidget (Variables|Default|GetHorizontalOverdriveMeter))))
                (Class|WBPPilgrimCoreOverdriveMeterHorizontal|UpdateOverdriveUIColor _aswbp_pilgrim_core_overdrive_meter_horizontal (Utilities|Struct|BreakPilgrimEventPilgrimSettingsUseProColorsforAllPartsChanged _payload_7))))))))
    (:then_7
      ;[1 1 1] Listening to Lefty Flip Setting Change Event
      (bind (_asynctaskproxy_11 _payload_8) (GameplayEventRouter|StartListeningToEvent (Variables|Default|GetEventRouter) "/Script/PilgrimSettingsRuntime.PilgrimEvent_PilgrimSettings_UseGuitarLeftyFlipChanged" (CallFunction|GetOwningPlayer))
        (:then
          (Utilities|Array|Add (Variables|Default|GetGameplayEventHandles) _asynctaskproxy_11))
        (:OnEventReceived
          (Variables|Default|SetIsLeftyFlip (Utilities|Struct|BreakPilgrimEventPilgrimSettingsUseGuitarLeftyFlipChanged _payload_8)))))))

(event Custom|CustomEvent_0
  (params (bNewInputType :type ECommonInputType%20Enum :default "MouseAndKeyboard"))
  ;[1 1 1] Listen for input  changes to adjust the divider, if necessary
  (CallFunction|SetTrackMaterial true))

(event Custom|OnInputSettingsChanged
  ;[1 1 1] Listen for input  changes to adjust the divider, if necessary
  (CallFunction|SetTrackMaterial true))

(event EventDestroyed
  (bind _gameplayeventhandles (Variables|Default|GetGameplayEventHandles))
  (for _array_element _gameplayeventhandles
    (GameplayEventRouter|StopListeningtoEvent _array_element))
  (Utilities|Array|Clear _gameplayeventhandles))

(event Custom|SetupOverdriveVisibility
  (bind _pilgrimgame (Variables|PilgrimTrack|GetPilgrimGame))
  ;[1 1 1] Overdrive Bar Visibility
  (Utilities|IsValid _pilgrimgame
    (:Is%20Valid
      (EventDispatchers|BindEventtoOnOverdriveHiddenChanged _pilgrimgame (delegate Custom|Update%20Overdrive%20Hidden))
      (bind _returnvalue (Class|PilgrimGame|IsOverdriveHidden _pilgrimgame))
      (CallFunction|UpdateOverdriveHidden _returnvalue))
    (:Is%20Not%20Valid
      (Development|PrintString "Unable to set up overdrive meter. Parent Game was invalid." true true "(R=1.000000,G=0.000000,B=0.568284,A=1.000000)" 4.0))))

(event Custom|Update%20Overdrive%20Hidden
  (params (IsOverdriveHidden :type Boolean))
  ;[1 1 1] Overdrive Bar Visibility
  (Rendering|SetVisibility (Variables|Default|GetHorizontalOverdriveMeter) (not IsOverdriveHidden) false))

(fn AddComponentsToTrackCapture
  (Class|PilgrimTrackCaptureComponent|IncludeActorinCustomView (CallFunction|GetCaptureComponent) self self true))

(fn CanPhaseTrackAtSongStart
  (params (PilgrimGame :type Pilgrim%20Game%20Object%20Reference))
  (bind (_id _bisvalidgembreak _startbreakatbar _startbreakatms _breakbarduration _breakmsduration _endbreakatbar _endbreakatms _bisstartingbreak) (Utilities|Struct|BreakPilgrimGemBreak (Class|PilgrimGame|GetGemBreakforTrackDifficulty PilgrimGame (Class|PilgrimGame|GetSelectedTrackType PilgrimGame) (Class|PilgrimGame|GetSelectedDifficulty PilgrimGame))))
  ;[0.5121 1 0.869] Given a valid pilgrim game, the track can phase in if there is NOT a valid starting gem break happening
  (Utilities|IsValid PilgrimGame
    (:Is%20Valid
      (return (Math|Boolean|NANDBoolean _bisvalidgembreak _bisstartingbreak)))
    (:Is%20Not%20Valid
      (return true))))

(fn GetCorrectLaneIndex
  (params (Index :type Integer))
  (return (select (CallFunction|CanUseProDrumsLaneIndex) (Utilities|Select :Index Index :Option%200 3 :Option%201 0 :Option%202 1 :Option%203 2 :Option%204 4) (select (Variables|Default|GetIsLeftyFlip) (select (Variables|Default|GetIs5Lanes) (Utilities|Select :Index Index :Option%200 4 :Option%201 3 :Option%202 2 :Option%203 1 :Option%204 0) (Utilities|Select :Index Index :Option%200 3 :Option%201 2 :Option%202 1 :Option%203 0)) Index))))

(fn GetCorrectProDrumsLaneIndex
  (params (Index :type Integer))
  (return (select (Variables|Default|GetIsLeftyFlip) (Utilities|Select :Index Index :Option%200 2 :Option%201 1 :Option%202 0 :Option%203 3 :Option%204 4) (Utilities|Select :Index Index :Option%200 3 :Option%201 0 :Option%202 1 :Option%203 2 :Option%204 4))))

(fn GetDividerParams
  (bind _returnvalue (CallFunction|GetOwningPlayer))
  (Utilities|IsValid _returnvalue
    (:Is%20Valid
      (bind _asfort_player_controller_gameplay (Utilities|Casting|CastToFortPlayerControllerGameplay _returnvalue)
        (:then
          (if (CallFunction|ShouldShowLaneDivider (Actor|GetComponentbyClass _asfort_player_controller_gameplay "/PilgrimCore/Player/Components/BP_PilgrimPlayerControllerComponent.BP_PilgrimPlayerControllerComponent_C"))
            (bind _index (Default|GetDividerIndex (Actor|GetComponentbyClass _asfort_player_controller_gameplay "/PilgrimCore/Player/Components/BP_PilgrimPlayerControllerComponent.BP_PilgrimPlayerControllerComponent_C")))
            (if (>= _index 0)
              ;[1 1 1] Show divider at index
              (return (Math|Conversions|ToFloat(Integer) (- _index 1)) 1.0)
              (else
                _))
            (else
              _))
          ;[1 1 1] Hide Divider
          (return 0.0 0.0))
        (:CastFailed
          _)))
    (:Is%20Not%20Valid
      _)))

(fn GetLaneFeedbackMaterialParameter
  (params (ParamIndex :type Integer))
  (bind _pilgrimtrackmaterialparametercollection (Variables|PilgrimTrack|GetPilgrimTrackMaterialParameterCollection))
  (bind _returnvalue (Rendering|Material|GetVectorParameterValue _pilgrimtrackmaterialparametercollection :!class /Script/Engine.KismetMaterialLibrary))
  (return (Utilities|Array|Get(aref) (Rendering|Material|GetVectorParameterNames _pilgrimtrackmaterialparametercollection) ParamIndex) _returnvalue))

(fn OnPilgrimBattleOptionsChanged
  (params (Options :type Music%20Battle%20Options%20Structure))
  (bind (_trackvisible_4_c10ed51e42d29f6e2874c6bd9161a585 _trackangle_8_92e985ab40016303f33eec8e00c8b09c _trackyaw_28_7c1e2f114bb73910dddf3e90a2bb91c3 _trackhorizontal_19_23c1ab764812c1aa01711193b08ffd81 _trackvertical_21_ce07e93b46dd488304ccc1829aa9044f _trackdepth_26_c8474826401cf785c44cad83d698f7a6 _trackcollision_23_a9d8249543dd587c9bc090b326c9bea3 _showbeatmarkers_16_ba8a5f3d4d1004d1edf89db8b98d54a1 _pathtype_14_0633619d40634703fdfd899fb84113b0) (Utilities|Struct|BreakMusicBattleOptions Options))
  (Utilities|FlowControl|Sequence
    (:then_0
      (Rendering|SetVisibility (Variables|PilgrimTrack|GetTrackMeshComponent) (and (Variables|Default|GetIsTrackVisible) _trackvisible_4_c10ed51e42d29f6e2874c6bd9161a585) false))
    (:then_1
      (bind _asbp_now_bar (Utilities|Casting|CastToBP_NowBar (Variables|/Script/Engine.ChildActorComponent|GetChildActor (Variables|PilgrimTrack|GetNowBarChildActor))))
      (Class|BPNowBar|SetBeatMarkerPath _asbp_now_bar _pathtype_14_0633619d40634703fdfd899fb84113b0)
      (for _array_element (Variables|/PilgrimCore/Actors/BP_NowBar.BP_NowBar_C|GetAllSmashers _asbp_now_bar)
        (Class|BPGemSmasher|SetGemPath _array_element _pathtype_14_0633619d40634703fdfd899fb84113b0)))))

(fn OwningPlayerHasEmoteDuringGemBreakComponent
  (bind _returnvalue (Actor|GetOwner))
  (bind _returnvalue_2 (Fortnite|GetControllerfromActor _returnvalue))
  ;[0.5121 1 0.869] Determine if the player controller has an EmoteDuringGemBreak component attached (used to determine when we phase the track in or out)
  (Utilities|IsValid _returnvalue_2
    (:Is%20Valid
      (return (Utilities|IsValid (Actor|GetComponentbyClass _returnvalue_2 "/PilgrimCore/Player/Components/BP_PilgrimEmoteDuringGemBreakPlayerControllerComponent.BP_PilgrimEmoteDuringGemBreakPlayerControllerComponent_C") :!class /Script/Engine.KismetSystemLibrary)))
    (:Is%20Not%20Valid
      (return false))))

(fn ReinitializeWidgetComponents
  (Utilities|FlowControl|Sequence
    (:then_0
      (bind _aswbp_streak_meter (Utilities|Casting|CastToWBP_StreakMeter (UserInterface|GetWidget (Variables|Default|GetStreakMeter))))
      (MusicBattle|AssignWidgettoPlayerController _aswbp_streak_meter (CallFunction|GetOwningPlayer))
      (Class|WBPStreakMeter|EventOnSetOwner _aswbp_streak_meter))
    (:then_1
      (bind _aswbp_pilgrim_core_overdrive_meter_horizontal (Utilities|Casting|CastToWBP_PilgrimCore_OverdriveMeter_Horizontal (UserInterface|GetWidget (Variables|Default|GetHorizontalOverdriveMeter))))
      (MusicBattle|AssignWidgettoPlayerController _aswbp_pilgrim_core_overdrive_meter_horizontal (CallFunction|GetOwningPlayer))
      (Class|WBPPilgrimCoreOverdriveMeterHorizontal|EventonOwnerSet _aswbp_pilgrim_core_overdrive_meter_horizontal))))

(fn SetLaneFeedbackMaterialParameter
  (params (ParamIndex :type Integer) (Param%20Name :type Name) (ParamValue :type Linear%20Color%20Structure))
  (Rendering|Material|SetVectorParameterValue (Variables|Default|GetDynamicMaterialInstanceRef) Param%20Name ParamValue :!class /Script/Engine.MaterialInstanceDynamic))

(fn SetPlasticParams
  ;[1 1 1] Currently used to set lane contrast for 5 lane without divider, can be used for other art within material if wanted
  (Rendering|Material|SetScalarParameterValue (Variables|Default|GetDynamicMaterialInstanceRef) "IsPlastic" (select (Class|PilgrimGame|IsTrackUsingProColors (Variables|PilgrimTrack|GetPilgrimGame)) 1.0 0.0) :!class /Script/Engine.MaterialInstanceDynamic))

(fn SetTrackMaterial
  (params (IsAnimated :type Boolean))
  (bind (_dividerposition _dividervisibility) (Default|GetDividerParams))
  (bind _trackmeshcomponent (Variables|PilgrimTrack|GetTrackMeshComponent))
  (bind _pilgrimgame (Variables|PilgrimTrack|GetPilgrimGame))
  ;[1 1 1] Set material
  (Rendering|Material|SetMaterial _trackmeshcomponent 0 (select IsAnimated (Variables|Default|GetAnimatedTrackMaterialInstance) (Variables|Default|GetDefaultTrackMaterialInstance)))
  (Variables|Default|SetDynamicMaterialInstanceRef (Class|BTechArtGlobalLibrary|GetDynamicMaterialInstance _trackmeshcomponent))
  ;[1 1 1] Make sure peak state is set for new material
  (Rendering|Material|SetScalarParameterValue (Variables|Default|GetDynamicMaterialInstanceRef) "Max Streak On/Off" (Variables|Default|GetMax%20Streak%20Fade) :!class /Script/Engine.MaterialInstanceDynamic)
  ;[1 1 1] Make sure we select correct divider lanes after setting new material
  (Utilities|IsValid _pilgrimgame
    (:Is%20Valid
      (Rendering|Material|SetScalarParameterValue (Variables|Default|GetDynamicMaterialInstanceRef) "Is5Lane" (select (> (Class|PilgrimGame|GetNumLanes _pilgrimgame) 4) 1.0 0.0) :!class /Script/Engine.MaterialInstanceDynamic)
      (CallFunction|SetPlasticParams)
      (Rendering|Material|SetScalarParameterValue (Variables|Default|GetDynamicMaterialInstanceRef) "DividerPosition" _dividerposition :!class /Script/Engine.MaterialInstanceDynamic)
      (Rendering|Material|SetScalarParameterValue (Variables|Default|GetDynamicMaterialInstanceRef) "UseDivider" _dividervisibility :!class /Script/Engine.MaterialInstanceDynamic)
      (Rendering|Material|SetScalarParameterValue (Variables|Default|GetDynamicMaterialInstanceRef) "IsDrumKit" (Math|Conversions|ToFloat(Boolean) (Class|PilgrimGame|IsDrumPeripheralTrack (Variables|PilgrimTrack|GetPilgrimGame))) :!class /Script/Engine.MaterialInstanceDynamic))))

(fn Setup4Lane
  ;[1 1 1] Set Peak State and Overdrive Niagara Variable for Track Size
  (Niagara|SetNiagaraVariable(Float) (Variables|Default|GetNS_Track_PeakState) "Is5Lane" 0.0)
  (Transformation|SetRelativeLocation (Variables|Default|GetHorizontalOverdriveMeter) (Variables|4Lane|GetOverdriveMeter_4LaneLocation) false false)
  (Transformation|SetRelativeLocation (Variables|Default|GetStreakMeter) (Variables|4Lane|GetStreakMeter_4LaneLocation) false false)
  (Rendering|Material|SetScalarParameterValueonMaterials (Variables|PilgrimTrack|GetTrackMeshComponent) "Is5Lane" 0.0))

(fn Setup5Lane
  ;[1 1 1] Set Peak State and Overdrive Niagara Variable for Track Size
  (Niagara|SetNiagaraVariable(Float) (Variables|Default|GetNS_Track_PeakState) "Is5Lane" 1.0)
  (Transformation|SetRelativeLocation (Variables|Default|GetHorizontalOverdriveMeter) (Variables|5Lane|GetOverdriveMeter_5LaneLocation) false false)
  (Transformation|SetRelativeLocation (Variables|Default|GetStreakMeter) (Variables|5Lane|GetStreakMeter_5LaneLocation) false false)
  (Rendering|Material|SetScalarParameterValueonMaterials (Variables|PilgrimTrack|GetTrackMeshComponent) "Is5Lane" 1.0))

(fn SetupIsPlasticDrums
  (Variables|Default|SetIsPlasticDrums (Class|PilgrimGame|IsDrumPeripheralTrack (Variables|PilgrimTrack|GetPilgrimGame))))

(fn SetupLanes_ChildMeters
  (params (Has5Lanes :type Boolean)))

(fn SetupLeftyFlip
  (Variables|Default|SetIsLeftyFlip (Class|PilgrimGame|IsUsingLeftyFlip (Variables|PilgrimTrack|GetPilgrimGame))))

(fn SetupMeterVisibilityByGameMode
  (params (IsPilgrimArena :type Boolean)))

(fn SetupNumLanes
  (params (NumLanes :type Integer))
  (bind _returnvalue (== NumLanes 5))
  (bind _asbp_now_bar (Utilities|Casting|CastToBP_NowBar (Variables|/Script/Engine.ChildActorComponent|GetChildActor (Variables|PilgrimTrack|GetNowBarChildActor))))
  (Class|BPNowBar|SetupNumLanes _asbp_now_bar NumLanes)
  (Variables|Default|SetIs5Lanes _returnvalue)
  (Components|StaticMesh|SetStaticMesh (Variables|PilgrimTrack|GetTrackMeshComponent) (select _returnvalue (Variables|5Lane|GetTrackMesh_5Lane) (Variables|4Lane|GetTrackMesh_4Lane)))
  (Utilities|FlowControl|Sequence
    (:then_0
      (if _returnvalue
        (CallFunction|Setup5Lane)
        ;[1 0.349 0.349] Delete these after setting up correct art for the track body and rails
        (Utilities|FlowControl|Sequence
          (:then_0
            (Transformation|AddRelativeLocation (Variables|PilgrimTrack|GetNowBarChildActor) "-2.000000,0.000000,0.000000" false true)))
        (else
          (CallFunction|Setup4Lane))))
    (:then_1
      (Default|SetupLanesChildMeters _returnvalue))))

(fn SetupOverdriveMeter
  (Rendering|SetVisibility (Variables|Default|GetHorizontalOverdriveMeter) (not (Variables|Default|GetShouldHideHorizontalOverdriveMeter)) false))

(fn SetupStreakMeter)

(fn ShouldShowLaneDivider
  (params (Controller :type BP%20Pilgrim%20Player%20Controller%20Component%20Object%20Reference))
  (return (and (Math|Boolean|ORBoolean (not (Class|PilgrimGame|IsTrackUsingProColors (Variables|PilgrimTrack|GetPilgrimGame))) (Class|PilgrimPlayerControllerComponent|IsUsingGamepad Controller) (Default|IsUsingMouseandKeyboard Controller)) (not (Class|PilgrimPlayerControllerComponent|GetIsCurrentDevicePeripheral Controller)))))

(fn TriggerLaneFeedback
  (params (LaneIndex :type EPilgrimTrackLane%20Enum :default "TrackLaneNone") (Accuracy :type Pilgrim%20Gem%20Hit%20Accuracy%20Structure))
  (bind (_accuracyms _accuracytier) (Utilities|Struct|BreakPilgrimGemHitAccuracy Accuracy))
  (bind (_thresholdms _scoremultiplier _name) (Utilities|Struct|BreakPilgrimAccuracyTier _accuracytier))
  (bind _returnvalue (CallFunction|GetShowDetailedHitFeedback))
  (bind _localcorrectlaneindex (LocalVariables|GetLocalCorrectLaneIndex))
  (bind _trackmeshcomponent (Variables|PilgrimTrack|GetTrackMeshComponent))
  (Utilities|FlowControl|Sequence
    (:then_0
      ;[1 1 1] Map Accuracy to Float Value
      (LocalVariables|SetAccuracyValue (select (Utilities|Text|EqualExactly(Text) _name (LocalVariables|GetGoodText)) (select (< _accuracyms 0) 1.0 -1.0) 0.0)))
    (:then_1
      (if (!= _returnvalue (LocalVariables|GetLastPerLaneFeedbackState))
        (bind _output_get (LocalVariables|SetLastPerLaneFeedbackState _returnvalue))
        (Rendering|Material|SetScalarParameterValue (Variables|Default|GetDynamicMaterialInstanceRef) "FeedbackOpacity" (select _output_get 1.0 0.0) :!class /Script/Engine.MaterialInstanceDynamic))
      (if (LocalVariables|GetLastPerLaneFeedbackState)
        ;[0 0.05494 0.15 0.5] Uses an index that gets injected into a material to know where to play the perfect accuracy vfx
        (bind _output_get_2 (LocalVariables|SetLocalLanePressedIndex (Utilities|Select :Index LaneIndex :TrackLane0 0 :TrackLane1 1 :TrackLane2 2 :TrackLane3 3 :TrackLane4 4 :TrackLane5 5)))
        (LocalVariables|SetLocalCorrectLaneIndex (Default|GetCorrectLaneIndex _output_get_2))
        (Rendering|Material|SetCustomPrimitiveDataFloat _trackmeshcomponent _localcorrectlaneindex (Utilities|Time|GetGameTimeinSeconds))
        (Rendering|Material|SetCustomPrimitiveDataFloat _trackmeshcomponent (+ _localcorrectlaneindex 5) (LocalVariables|GetAccuracyValue))))))

(fn UpdateTrackWidth
  ;[0.04814 0.1149 0.15 0.5] If we're on mobile and UseMobileExpandedTrack CVar is enabled, then the track is scaled on the X axis by 2.
  ;
  ; The initial track width is factored into the scale determined by UPilgrimTrackAndGameManagerComponent::SpawnTrack
  (Transformation|SetRelativeScale3D (Variables|PilgrimTrack|GetTrackMeshComponent) (* (Variables|/Script/PilgrimCoreRuntime.PilgrimTrackAndGameManagerComponent|GetTrackScale (Variables|PilgrimTrack|GetTrackAndGameManagerComponent)) (select (Class|PilgrimCoreBPFL|IsMobileandShouldUseExpandedTrack) "2.000000,1.000000,1.000000" "1.000000,1.000000,1.000000"))))

(fn UpdateLaneFeedbackOpacity
  (params (FadeSpeed :type Float%20%28double-precision%29 :default 1.0) (Delta :type Float%20%28double-precision%29))
  (bind _returnvalue (* FadeSpeed Delta))
  (bind _lanefeedbackopacity (Variables|Default|GetLaneFeedbackOpacity))
  (for (_ _array_index) (Variables|Default|GetLaneFeedbackOpacity)
    (if (> (Utilities|Array|Get(aref) _lanefeedbackopacity (Default|GetCorrectLaneIndex _array_index)) 0)
      (Utilities|Array|SetArrayElem (Variables|Default|GetLaneFeedbackOpacity) (Default|GetCorrectLaneIndex _array_index) (Math|Float|Clamp(Float) (- (Utilities|Array|Get(aref) _lanefeedbackopacity (Default|GetCorrectLaneIndex _array_index)) _returnvalue)))
      (CallFunction|SetLaneFeedbackMaterialParameter (Default|GetCorrectLaneIndex _array_index) (Utilities|Name|MakeLiteralName (Utilities|String|StringToName (Utilities|String|Append "HitFeedback_Lane" (Utilities|String|ToString(Integer) (+ (Default|GetCorrectLaneIndex _array_index) 1))))) (Math|Color|MakeColor (select (Utilities|Text|EqualExactly(Text) (.Name (Utilities|Struct|BreakPilgrimAccuracyTier (.AccuracyTier (Utilities|Struct|BreakPilgrimGemHitAccuracy (Utilities|Array|Get(aref) (Variables|Default|GetLaneFeedback) (Default|GetCorrectLaneIndex _array_index)))))) (LocalVariables|GetGoodText)) 0.0 1.0) 0.0 (select (< (Utilities|Struct|BreakPilgrimGemHitAccuracy (Utilities|Array|Get(aref) (Variables|Default|GetLaneFeedback) (Default|GetCorrectLaneIndex _array_index))) 0) 0.0 1.0) (Math|Float|Clamp(Float) (- (Utilities|Array|Get(aref) _lanefeedbackopacity (Default|GetCorrectLaneIndex _array_index)) _returnvalue)))))))

(fn ConstructionScript
  ;[0.5121 1 0.869] On consctruction, set the visibility of the track body and secondary elements to invisible by default. This is to support the track not popping in when a song starts with a gem break.
  (CallFunction|SetSecondaryElementsPhase 0.0)
  (Rendering|Material|SetScalarParameterValueonMaterials (Variables|PilgrimTrack|GetTrackMeshComponent) "TrackPhaseInOut" 1.0))
