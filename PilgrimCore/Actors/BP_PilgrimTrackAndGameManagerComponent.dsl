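; Pilgrim music game component that manages how the note tracks are framed and captured to a render target.
; Layout positions the scene capture, registers tracks as view actors, then sets FOV, a cropped custom projection matrix and the texture target.
; Gameplay override configs supply four- or five-lane camera transform and FOV, falling back to track defaults.
; On mobile, router events refresh track scale and render target on expanded-track changes, and camera transform when songs play.
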(event EventBeginPlay
  (if (TouchPlatform|IsMobileGame)
    (bind _eventrouter (Pilgrim|PilgrimWaitForEventRouter (Components|GetOwner)))
    (Utilities|FlowControl|Sequence
      (:then_0
        (bind _asynctaskproxy (GameplayEventRouter|StartListeningToEvent _eventrouter "/Script/PilgrimSettingsRuntime.PilgrimEvent_PilgrimSettings_EnableMobileExpandedTrackChanged")
          (:then
            (Utilities|Array|Add (Variables|Default|GetGameplayEventHandles) _asynctaskproxy))
          (:OnEventReceived
            (CallFunction|UpdateTrackScale)
            (CallFunction|SetupRenderTarget))))
      (:then_1
        (bind (_asynctaskproxy_2 _payload) (GameplayEventRouter|StartListeningToEvent _eventrouter "/Script/PilgrimCoreRuntime.PilgrimEvent_PilgrimGame_PlayingSong")
          (:then
            (Utilities|Array|Add (Variables|Default|GetGameplayEventHandles) _asynctaskproxy_2))
          (:OnEventReceived
            (CallFunction|SetupCaptureComponentRelativeTransform (Utilities|Struct|BreakPilgrimEventPilgrimGamePlayingSong _payload))))))))

(event EventLayoutTracksAndCaptureComponent
  (Utilities|FlowControl|Sequence
    (:then_0
      (CallFunction|SetupCaptureComponentRelativeTransform 0))
    (:then_1
      (CallFunction|SetupViewActors))
    (:then_2
      (CallFunction|SetupRenderTarget))))

(event EventSetupRenderTarget
  (CallFunction|SetupRenderTargetandProjection)
  (|CallParentFunction))

(event EventEndPlay
  (params (EndPlayReason :type EEndPlayReason%20Enum :default "Destroyed"))
  (bind _gameplayeventhandles (Variables|Default|GetGameplayEventHandles))
  (for _array_element _gameplayeventhandles
    (GameplayEventRouter|StopListeningtoEvent _array_element))
  (Utilities|Array|Clear _gameplayeventhandles))

(fn IsFirstPilgrimGameUsingFiveLanes
  (params (PilgrimGameIndex :type Integer))
  (bind _pilgrimgames (Variables|PilgrimTrackAndGameManagerComponent|GetPilgrimGames))
  (bind _output (Utilities|Array|Get(acopy) _pilgrimgames PilgrimGameIndex))
  (if (Utilities|Array|IsNotEmpty _pilgrimgames)
    (Utilities|IsValid _output
      (:Is%20Valid
        (return (== (Class|PilgrimGame|GetNumLanes _output) 5)))
      (:Is%20Not%20Valid
        _))
    (else
      _))
  (return false))

(fn IsPilgrimGameUsingFiveLanes
  (params (InPilgrimGame :type Pilgrim%20Game%20Object%20Reference))
  (Utilities|IsValid InPilgrimGame
    (:Is%20Valid
      (if (Utilities|Array|ContainsItem (Variables|PilgrimTrackAndGameManagerComponent|GetPilgrimGames) InPilgrimGame)
        (return (== (Class|PilgrimGame|GetNumLanes InPilgrimGame) 5))
        (else
          _)))
    (:Is%20Not%20Valid
      _))
  (return false))

(fn OverrideCaptureTransformWithConfig
  (params (PilgrimCaptureConfigOverride :type Pilgrim%20Capture%20Config%20Override%20Structure%20%28by%20ref%29))
  (bind _capturetransformoverride (Utilities|Struct|BreakPilgrimCaptureConfigOverride PilgrimCaptureConfigOverride))
  (bind _capturecomp (Variables|PilgrimTrackAndGameManagerComponent|GetCaptureComp))
  (Transformation|SetWorldLocation _capturecomp (+ (Variables|PilgrimTrack|GetTrackSceneStartingPos) (* (.location _capturetransformoverride) (Variables|PilgrimTrack|GetTrackScale))) false false)
  (Transformation|SetWorldRotation _capturecomp (.rotation _capturetransformoverride) false false))

(fn SetupCaptureComponentRelativeTransform
  (params (PilgrimGame :type Pilgrim%20Game%20Object%20Reference))
  (bind _returnvalue (CallFunction|GetGameplayOverridesConfig))
  (bind _capturecomp (Variables|PilgrimTrackAndGameManagerComponent|GetCaptureComp))
  (bind _fourlanecaptureconfigoverride (Variables|/Script/PilgrimCoreRuntime.PilgrimMusicBattleInstanceGameplayOverridesConfig|GetFourLaneCaptureConfigOverride _returnvalue))
  (bind _fivelanecaptureconfigoverride (Variables|/Script/PilgrimCoreRuntime.PilgrimMusicBattleInstanceGameplayOverridesConfig|GetFiveLaneCaptureConfigOverride _returnvalue))
  (bind _pilgrimgame (Variables|GetPilgrimGame))
  (Utilities|IsValid _returnvalue
    (:Is%20Valid
      (Utilities|IsValid _pilgrimgame
        (:Is%20Valid
          (LocalVariables|SetLocalCaptureConfigOverride (select (CallFunction|IsPilgrimGameUsingFiveLanes _pilgrimgame) _fivelanecaptureconfigoverride _fourlanecaptureconfigoverride)))
        (:Is%20Not%20Valid
          (LocalVariables|SetLocalCaptureConfigOverride (select (CallFunction|IsFirstPilgrimGameUsingFiveLanes 0) _fivelanecaptureconfigoverride _fourlanecaptureconfigoverride))))
      (CallFunction|OverrideCaptureTransformwithConfig (LocalVariables|GetLocalCaptureConfigOverride)))
    (:Is%20Not%20Valid
      (Transformation|SetWorldLocation _capturecomp (+ (Variables|PilgrimTrack|GetTrackSceneStartingPos) (* (Variables|PilgrimTrack|GetTrackSceneOffset) (Variables|PilgrimTrack|GetTrackScale))) false false)
      (Transformation|SetWorldRotation _capturecomp (Variables|PilgrimTrack|GetTrackSceneRotation) false false))))

(fn SetupFOV
  (bind (_capturetransformoverride _capturefovoverride) (Utilities|Struct|BreakPilgrimCaptureConfigOverride (select (CallFunction|IsFirstPilgrimGameUsingFiveLanes 0) (Variables|/Script/PilgrimCoreRuntime.PilgrimMusicBattleInstanceGameplayOverridesConfig|GetFiveLaneCaptureConfigOverride (CallFunction|GetGameplayOverridesConfig)) (Variables|/Script/PilgrimCoreRuntime.PilgrimMusicBattleInstanceGameplayOverridesConfig|GetFourLaneCaptureConfigOverride (CallFunction|GetGameplayOverridesConfig)))))
  (bind _returnvalue (CallFunction|GetGameplayOverridesConfig))
  (LocalVariables|SetFOV (Variables|PilgrimTrack|GetTrackSceneFOV))
  (Utilities|FlowControl|Sequence
    (:then_0
      (Utilities|IsValid _returnvalue
        (:Is%20Valid
          (LocalVariables|SetFOV _capturefovoverride))))
    (:then_1
      (Variables|/Script/Engine.SceneCaptureComponent2D|SetFOVAngle (LocalVariables|GetFOV) (Variables|PilgrimTrackAndGameManagerComponent|GetCaptureComp)))))

(fn SetupRenderTargetAndProjection
  (bind _capturecomp (Variables|PilgrimTrackAndGameManagerComponent|GetCaptureComp))
  (bind _outfade (Class|PilgrimTrackCaptureComponent|GetTrackCaptureFade (Variables|PilgrimTrackAndGameManagerComponent|GetCaptureComp) (Utilities|Array|Length (Variables|PilgrimTrackAndGameManagerComponent|GetTracks))))
  (CallFunction|SetupFOV)
  (bind _returnvalue (CallFunction|CropTargetandGetProjectionMatrix (Math|Vector4|MakeVector4 (Math|Vector2D|BreakVector2D (Utilities|Struct|BreakPilgrimTrackCaptureBox (.Crop (Utilities|Struct|BreakPilgrimTrackCaptureFade _outfade)))) (.Y (Math|Vector2D|BreakVector2D (Utilities|Struct|BreakPilgrimTrackCaptureBox (.Crop (Utilities|Struct|BreakPilgrimTrackCaptureFade _outfade))))) (Math|Vector2D|BreakVector2D (.RightBottom (Utilities|Struct|BreakPilgrimTrackCaptureBox (.Crop (Utilities|Struct|BreakPilgrimTrackCaptureFade _outfade))))) (.Y (Math|Vector2D|BreakVector2D (.RightBottom (Utilities|Struct|BreakPilgrimTrackCaptureBox (.Crop (Utilities|Struct|BreakPilgrimTrackCaptureFade _outfade))))))) (Variables|/Script/PilgrimCoreRuntime.PilgrimTrackCaptureLayoutDefinitions|GetbUseCropAspectRatio (Variables|/Script/PilgrimCoreRuntime.PilgrimTrackCaptureComponent|GetLayoutDefinitions (Variables|PilgrimTrackAndGameManagerComponent|GetCaptureComp))) (.bHasCropAspectRatioOverride (Utilities|Struct|BreakPilgrimTrackCaptureFade _outfade)) (.CropAspectRatioOverride (Utilities|Struct|BreakPilgrimTrackCaptureFade _outfade)) (Utilities|Struct|BreakPilgrimTrackCaptureFade _outfade)))
  (Variables|/Script/Engine.SceneCaptureComponent2D|SetCustomProjectionMatrix _returnvalue _capturecomp)
  (Variables|/Script/Engine.SceneCaptureComponent2D|SetbUseCustomProjectionMatrix true _capturecomp)
  (Class|PilgrimTrackCaptureComponent|SetAuthoredTransform (Variables|PilgrimTrackAndGameManagerComponent|GetCaptureComp) (+ (Variables|PilgrimTrack|GetTrackSceneStartingPos) (* (Variables|PilgrimTrack|GetTrackSceneOffset) (Variables|PilgrimTrack|GetTrackScale))) (Variables|PilgrimTrack|GetTrackSceneRotation) (CallFunction|GetCroppedProjectionMatrix (Variables|PilgrimTrack|GetTrackSceneFOV)))
  (bind _returnvalue_2 (CallFunction|GetRenderTarget))
  (Variables|/Script/Engine.SceneCaptureComponent2D|SetTextureTarget _returnvalue_2 (Variables|PilgrimTrackAndGameManagerComponent|GetCaptureComp)))

(fn SetupViewActors
  (Class|PilgrimTrackCaptureComponent|ClearViewActors (Variables|PilgrimTrackAndGameManagerComponent|GetCaptureComp))
  (for (_array_element _array_index) (Variables|PilgrimTrackAndGameManagerComponent|GetTracks)
    (Utilities|FlowControl|Sequence
      (:then_0
        (if (> _array_index 0)
          (Transformation|SetWorldLocation (Variables|/Script/PilgrimCoreRuntime.PilgrimTrack|GetTrackMeshComponent _array_element) (Variables|PilgrimTrack|GetTrackSceneStartingPos) false false)))
      (:then_1
        (Class|PilgrimTrackCaptureComponent|AddViewActor (Variables|PilgrimTrackAndGameManagerComponent|GetCaptureComp) _array_element)))))
