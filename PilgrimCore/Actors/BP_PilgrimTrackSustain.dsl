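; BP_PilgrimTrackSustain is a pooled track sustain actor that draws a held note's trail on the rhythm-game track.
; BeginPlay creates spline meshes for the track segments plus start and end caps.
; When added to a track, it resets body materials and, for RollsAndSwells trails, configures caps, width, end padding and now-bar collapse.
; Press, release, fail and modulation events drive body modulation, intensity, glow and width through custom primitive data.
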(event EventBeginPlay
  (CallFunction|CreateEnoughSplineMeshes))

(event EventOnAddedToTrack
  (bind (_starttick _endtick _numscorables _ms _liftnotesustainms _durationms _liftnotesustaindurationms _durationticks _track _lane _beattype _gemtype _trackitemindex _isoverdrivegem _bisdoublekickgem _overdrivesectionnum _isendofoverdrivesection _ballowhitonpress _ballowhitonrelease _bissustain _sustaintrailtype) (Utilities|Struct|BreakTrackItemInfo (Variables|PilgrimTrackItem|GetItemInfo)))
  ;[1 1 1] Init this sustain from the actor pool. Note this sustain will get reused.  Set Intensity and overdrive sustain here.
  (|CallParentFunction)
  (Utilities|FlowControl|Sequence
    (:then_0
      (for _array_element (Variables|PilgrimTrackSustain|GetCreatedSplineMeshes)
        (CallFunction|SetMaterialParams _array_element 0.0 0.4 0.05 0.99) ; Initial values
      ))
    (:then_1
      (switch Utilities|FlowControl|Switch|SwitchonEPilgrimSustainTrailType _sustaintrailtype
        (:RollsAndSwells
          (CallFunction|SetSustainCapVisibility (Variables|PilgrimTrackSustain|GetStartCapSplineMesh) true true)
          (CallFunction|SetSustainCapVisibility (Variables|PilgrimTrackSustain|GetEndCapSplineMesh) true)
          (Variables|PilgrimTrackSustain|SetSustainWidthScale 1.5)
          (Variables|PilgrimTrackSustain|SetSustainBodyEndPosPaddingMs -1e+02)
          (Variables|PilgrimTrackSustain|SetbCollapseAtNowBarEnabled true))))))

(event EventOnSustainFail
  ;[1 1 1] Input Fail
  (|CallParentFunction)
  (for _array_element (Variables|PilgrimTrackSustain|GetCreatedSplineMeshes)
    (CallFunction|SetMaterialParams _array_element 0.0 0.2 0.0 0.99)))

(event EventOnSustainKeyDownSuccess
  (params (CurrentModulation :type Float%20%28single-precision%29))
  ;[1 1 1] Sustain Input Pressed (processed when the corresponding gem is smashed)
  (|CallParentFunction CurrentModulation)
  (for _array_element (Variables|PilgrimTrackSustain|GetCreatedSplineMeshes)
    (CallFunction|SetMaterialParams _array_element CurrentModulation 1.0 0.2 0.96)))

(event EventOnSustainKeyRelease
  ;[1 1 1] Input Release
  (|CallParentFunction)
  (for _array_element (Variables|PilgrimTrackSustain|GetCreatedSplineMeshes)
    (CallFunction|SetMaterialParams _array_element 0.0 0.2 0.0 0.99)))

(event EventOnModulationChanged
  (params (CurrentModulation :type Float%20%28single-precision%29))
  ;[1 1 1] Modulation Changed
  (CallFunction|ApplyModulation CurrentModulation))

(fn ApplyModulation
  (params (CurrentModulation :type Float%20%28double-precision%29))
  (for _array_element (Variables|PilgrimTrackSustain|GetCreatedSplineMeshes)
    (Rendering|Material|SetCustomPrimitiveDataFloat _array_element 8 CurrentModulation)))

(fn CreateEnoughSplineMeshes
  (Utilities|FlowControl|Sequence
    (:then_0
      (if (< (Utilities|Array|Length (Variables|PilgrimTrackSustain|GetCreatedSplineMeshes)) (Variables|PilgrimTrackSustain|GetNumTrackSegments))
        (for _ (range (Variables|PilgrimTrackSustain|GetNumTrackSegments))
          (bind _returnvalue (AddComponent|NODE_AddSplineMeshComponent-0))
          (Components|StaticMesh|SetStaticMesh _returnvalue (Variables|Default|GetSplineStaticMesh))
          (Rendering|Material|SetMaterial _returnvalue 0 (Variables|Default|GetMaterial))
          (Utilities|Array|Add (Variables|PilgrimTrackSustain|GetCreatedSplineMeshes) _returnvalue))))
    (:then_1
      (bind _returnvalue_2 (AddComponent|NODE_AddSplineMeshComponent-1))
      (Components|StaticMesh|SetStaticMesh _returnvalue_2 (Variables|Default|GetSplineStaticMesh))
      (Rendering|Material|SetMaterial _returnvalue_2 0 (Variables|Default|GetMaterial))
      (Variables|PilgrimTrackSustain|SetStartCapSplineMesh _returnvalue_2))
    (:then_2
      (bind _returnvalue_3 (AddComponent|NODE_AddSplineMeshComponent-2))
      (Components|StaticMesh|SetStaticMesh _returnvalue_3 (Variables|Default|GetSplineStaticMesh))
      (Rendering|Material|SetMaterial _returnvalue_3 0 (Variables|Default|GetMaterial))
      (Variables|PilgrimTrackSustain|SetEndCapSplineMesh _returnvalue_3))))

(fn SetMaterialParams
  (params (Component :type Primitive%20Component%20Object%20Reference) (Modulation :type Float%20%28double-precision%29) (Intensity :type Float%20%28double-precision%29) (Glow :type Float%20%28double-precision%29) (Width :type Float%20%28double-precision%29))
  (Rendering|Material|SetCustomPrimitiveDataVector4 Component 8 (Math|Vector4|MakeVector4 Modulation Intensity Glow Width)) ; Set all 4 params at the same time:\n8: Modulation\n9: Intensity\n10: Glow\n11: Width
)

(fn SetSustainCapVisibility
  (params (SustainCapMesh :type Spline%20Mesh%20Component%20Object%20Reference) (IsVisible :type Boolean) (IsUVFlipped :type Boolean))
  (Rendering|SetVisibility SustainCapMesh IsVisible false)
  (if IsVisible
    (CallFunction|SetMaterialParams SustainCapMesh 0.0 0.4 0.05 0.99) ; Initial values
    (Rendering|Material|SetCustomPrimitiveDataFloat SustainCapMesh 12 1.0)
    (Rendering|Material|SetCustomPrimitiveDataFloat SustainCapMesh 13 (select IsUVFlipped 1.0 0.0))))

(fn ConstructionScript)
