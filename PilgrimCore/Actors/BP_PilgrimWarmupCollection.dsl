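; Shader warmup actor that renders configured materials, meshes and Niagara systems so their shaders compile.
; Registers as a component-manager receiver, and Warmup gathers extra entries from components implementing the warmup interface.
; It builds static, spline or instanced mesh components by material usage (skeletal unsupported), spawns Niagara systems, and captures through a dummy track actor using the owner's track-capture render target.
; VelocityPass nudges that dummy and recaptures to include the velocity pass.
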(event EventBeginPlay
  ; Register with Game Framework Component Manager
  (Gameplay|AddReceiver (GameInstanceSubsystems|GetGameFrameworkComponentManager) 0 true))

(event EventEndPlay
  (params (EndPlayReason :type EEndPlayReason%20Enum :default "Destroyed"))
  (Actor|DestroyActor (Variables|Default|GetSceneViewDummyActor))
  (Gameplay|RemoveReceiver (GameInstanceSubsystems|GetGameFrameworkComponentManager) 0))

(fn IsDoneProcessingMaterials
  (return (>= (Variables|Default|GetMaterialIndex) (Utilities|Array|Length (Variables|Default|GetMeshesAndMaterials)))))

(fn ConstructionScript)

(fn VelocityPass
  ; To get included in the Velocity pass we must move our actor.
  (Transformation|AddActorWorldOffset (Variables|Default|GetSceneViewDummyActor) "1.000000,0.000000,0.000000" false false)
  (Rendering|SceneCapture|CaptureScene (Variables|/PilgrimCore/Actors/BP_DummyTrack.BP_DummyTrack_C|GetSceneCaptureComponent2D (Variables|Default|GetSceneViewDummyActor)) :!class /Script/Engine.SceneCaptureComponent2D))

(fn Warmup
  (bind _localincludeinscenecapture (LocalVariables|GetLocalIncludeInSceneCapture))
  (bind _localmaterialparam (LocalVariables|GetLocalMaterialParam))
  (bind _sceneviewdummyactor (Variables|Default|GetSceneViewDummyActor))
  (bind _scenecapturecomponent2d (Variables|/PilgrimCore/Actors/BP_DummyTrack.BP_DummyTrack_C|GetSceneCaptureComponent2D _sceneviewdummyactor))
  (bind _sceneviewdummyactor_2 (Variables|Default|GetSceneViewDummyActor))
  (bind _scenecapturecomponent2d_2 (Variables|/PilgrimCore/Actors/BP_DummyTrack.BP_DummyTrack_C|GetSceneCaptureComponent2D _sceneviewdummyactor_2))
  (Utilities|FlowControl|Sequence
    (:then_0
      ; Get Track Render Component, to make the shaders compile correctly we need to render to the 2d scene component. We also need a dummy actor to add as a view target.
      (bind _output_get (Variables|Default|SetPilgrimTrackCaptureComponentCache (Actor|GetComponentbyClass (Actor|GetOwner) "/Script/PilgrimCoreRuntime.PilgrimTrackCaptureComponent")))
      (bind _returnvalue (Class|PilgrimTrackCaptureComponent|GetRenderTarget _output_get))
      (bind _returnvalue_2 (Game|SpawnActorfromClass "/PilgrimCore/Actors/BP_DummyTrack.BP_DummyTrack_C" (Transformation|GetActorTransform) "AlwaysSpawn" "OverrideRootScale" self _returnvalue 0))
      (bind _output_get_2 (Variables|Default|SetSceneViewDummyActor _returnvalue_2))
      (Development|PrintString (Utilities|String|ToString(Text) (Utilities|Text|FormatText "NSLOCTEXT(\"[396581885D87ACC040E88FF37904D8AC]\", \"6849FC10464D13C302939E88844FA934\", \"Adding View Actor {Name} to pilgrim track capture component {Component} with capture source {Source}\")" (Utilities|GetDisplayName 0) (Utilities|GetDisplayName _output_get_2) (Utilities|String|EnumtoString (Variables|/Script/Engine.SceneCaptureComponent|GetCaptureSource _output_get))))))
    (:then_1
      ; Grab components from other plugins that may want things preloaded. Find all components on this actor that implement the preload interface and add their materials as well.
      (for _array_element (Actor|GetComponentsbyInterface "/PilgrimCore/Interfaces/BP_PilgrimShaderWarmupInterface.BP_PilgrimShaderWarmupInterface_C")
        (Utilities|IsValid _array_element
          (:Is%20Valid
            (Utilities|FlowControl|Sequence
              (:then_0
                (bind _warmupcollection (Class|BPPilgrimShaderWarmupInterface|GetMaterialsandMeshes(Message) _array_element))
                (Utilities|Array|AppendArray (Variables|Default|GetMeshesAndMaterials) _warmupcollection))
              (:then_1
                (bind _niagarasystems (Class|BPPilgrimShaderWarmupInterface|GetNiagaraSystems(Message) _array_element))
                (Utilities|Array|AppendArray (Variables|Default|GetNiagaraSystems) _niagarasystems)))))))
    (:then_2
      ; Get each material instance, and determine the correct mesh type, then create the mesh, set the material, and render on the track capture component.
      (for _array_element_2 (Variables|Default|GetMeshesAndMaterials)
        (Utilities|FlowControl|Sequence
          (:then_0
            (LocalVariables|SetLocalMaterialParam (Utilities|Struct|BreakPilgrimShaderWarmupCollection _array_element_2))
            (LocalVariables|SetLocalMeshParam (.Meshes_12_1A652E424A07B1CCC57F2DB53A994B2E (Utilities|Struct|BreakPilgrimShaderWarmupCollection _array_element_2)))
            (LocalVariables|SetLocalIncludeInSceneCapture (.IncludeInSceneCapture_16_52E4AEEE4418AE1C07880FBB51D15694 (Utilities|Struct|BreakPilgrimShaderWarmupCollection _array_element_2))))
          (:then_1
            (bind _returnvalue_3 (Rendering|Material|GetBaseMaterial (LocalVariables|GetLocalMaterialParam)))
            (bind _meshtypes (Class|PilgrimCoreUtilsLibrary|GetWarmupMeshTypes _returnvalue_3))
            (for _array_element_3 _meshtypes
              (Development|PrintString (Utilities|String|ToString(Text) (Utilities|Text|FormatText "NSLOCTEXT(\"[396581885D87ACC040E88FF37904D8AC]\", \"E11038DF4AE47EB0566C88809489BB24\", \"Warming Up Material {Mat} of type {Type}\")" (Utilities|GetDisplayName _returnvalue_3) (Utilities|String|EnumtoString _array_element_3))) true true "(R=0.000000,G=0.660000,B=1.000000,A=1.000000)" 1e+01)
              (for _array_element_4 (LocalVariables|GetLocalMeshParam)
                (switch Utilities|FlowControl|Switch|SwitchonEWarmupMeshType _array_element_3
                  (:NewEnumerator0
                    (Development|PrintString "Warmup of Skeletal Meshes are not currently supported, coming soon. "))
                  (:NewEnumerator5
                    _)
                  (:NewEnumerator6
                    (Development|PrintString "Niagara  coming from warmup materials. Add Niagara system to GetNiagaraSystems instead. "))
                  (:NewEnumerator8
                    _)
                  (:NewEnumerator9
                    _)
                  (:NewEnumerator11
                    (bind _returnvalue_4 (AddComponent|NODE_AddSplineMeshComponent-1))
                    (Utilities|IsValid _returnvalue_4
                      (:Is%20Valid
                        (Transformation|SetMobility _returnvalue_4 "Movable")
                        (bind _returnvalue_5 (Components|StaticMesh|SetStaticMesh _returnvalue_4 _array_element_4))
                        (if _returnvalue_5
                          (Rendering|Material|SetMaterial _returnvalue_4 0 _localmaterialparam)
                          ;; Add to the scene capture component if required.
                          (if _localincludeinscenecapture
                            (Rendering|SceneCapture|ShowOnlyComponent _scenecapturecomponent2d_2 _returnvalue_4)
                            (Rendering|SetVisibleinSceneCaptureOnly _returnvalue_4 true))))))
                  (:NewEnumerator12
                    (bind _returnvalue_6 (AddComponent|NODE_AddInstancedStaticMeshComponent-1))
                    (Utilities|IsValid _returnvalue_6
                      (:Is%20Valid
                        (bind _returnvalue_7 (Components|StaticMesh|SetStaticMesh _returnvalue_6 _array_element_4))
                        (if _returnvalue_7
                          (Rendering|Material|SetMaterial _returnvalue_6 0 _localmaterialparam)
                          (if _localincludeinscenecapture
                            (Rendering|SceneCapture|ShowOnlyComponent _scenecapturecomponent2d_2 _returnvalue_6)
                            (Rendering|SetVisibleinSceneCaptureOnly _returnvalue_6 true))))))
                  (:NewEnumerator13
                    _)
                  (:NewEnumerator20
                    _)
                  (:NewEnumerator24
                    _))
                (bind _returnvalue_8 (AddComponent|NODE_AddStaticMeshComponent-1))
                (Utilities|IsValid _returnvalue_8
                  (:Is%20Valid
                    (bind _returnvalue_9 (Components|StaticMesh|SetStaticMesh _returnvalue_8 _array_element_4))
                    (if _returnvalue_9
                      (Rendering|Material|SetMaterial _returnvalue_8 0 _localmaterialparam)
                      (if _localincludeinscenecapture
                        (Rendering|SceneCapture|ShowOnlyComponent _scenecapturecomponent2d_2 _returnvalue_8)
                        (Rendering|SetVisibleinSceneCaptureOnly _returnvalue_8 true)))))))))))
    (:then_3
      ; Load up Niagara systems
      (for _array_element_5 (Variables|Default|GetNiagaraSystems)
        (if (Utilities|IsValid (Utilities|Struct|BreakPilgrimShaderNiagaraWarmupCollection _array_element_5) :!class /Script/Engine.KismetSystemLibrary)
          (Development|PrintString (Utilities|String|ToString(Text) (Utilities|Text|FormatText "NSLOCTEXT(\"[396581885D87ACC040E88FF37904D8AC]\", \"7154563941E0945A0D252AB84D45600B\", \"Warming Up Niagara System {Name} with include in  Scene Capture {SceneCapture}\")" (Utilities|GetDisplayName (Utilities|Struct|BreakPilgrimShaderNiagaraWarmupCollection _array_element_5)) (.IncludeInSceneCapture_5_D91CC88248613A78CF3BADBC76497463 (Utilities|Struct|BreakPilgrimShaderNiagaraWarmupCollection _array_element_5)))))
          (bind _returnvalue_10 (Niagara|SpawnSystemAttached (Utilities|Struct|BreakPilgrimShaderNiagaraWarmupCollection _array_element_5) 0 "None" "0, 0, 0" "0, 0, 0" "KeepRelativeOffset" true true "None" false))
          (if (.IncludeInSceneCapture_5_D91CC88248613A78CF3BADBC76497463 (Utilities|Struct|BreakPilgrimShaderNiagaraWarmupCollection _array_element_5))
            (Rendering|SceneCapture|ShowOnlyComponent (Variables|/PilgrimCore/Actors/BP_DummyTrack.BP_DummyTrack_C|GetSceneCaptureComponent2D (Variables|Default|GetSceneViewDummyActor)) _returnvalue_10)
            (Rendering|SetVisibleinSceneCaptureOnly _returnvalue_10 true)))))
    (:then_4
      ; Capture Scene must be called at least once to render to the capture component. Capture every frame is turned off.
      (Rendering|SceneCapture|ShowOnlyActorComponents _scenecapturecomponent2d (Variables|Default|GetSceneViewDummyActor) true)
      (Rendering|SceneCapture|CaptureScene _scenecapturecomponent2d :!class /Script/Engine.SceneCaptureComponent2D))))
