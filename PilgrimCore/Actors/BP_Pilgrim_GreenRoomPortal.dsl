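; Pilgrim green room portal that shows an island's details and lets the party leader start or cancel matchmaking into it.
; The island code goes through FortPortalComponent, whose link data and thumbnail callbacks feed a proximity-shown HUD.
; Local interaction checks leadership and each controller's portal component, with ready-check handling and an optional cancel on leaving the island.
; A CVar gates the portal at BeginPlay, and cosmetic work runs client-side only.
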(event Custom|Cancel%20Matchmaking
  (params (Fort%20Player%20Controller :type Fort%20Player%20Controller%20Object%20Reference))
  ;[0 0.1399 1 0.5] Matchmaking Logic
  (bind _asfort_player_controller_athena (Utilities|Casting|CastToFortPlayerControllerAthena Fort%20Player%20Controller))
  (if (CallFunction|ClientOnlyIsPartyLeader _asfort_player_controller_athena)
    (Class|FortPlayerControllerAthena|CancelMatchmaking _asfort_player_controller_athena "CreativeCanceledByLeader")
    (Variables|Default|SetHasActiveReadyCheck false)))

(event OnLinkDataError(FortPortal)
  (params (ErrorCode :type String))
  ;[0 0.6421 1 0.5] Server Replicated Island Details Loading Logic
  (Utilities|BasicCodeValidation ErrorCode
    (:Valid%20Code
      (CallFunction|IslandDataNoSuccess "NewEnumerator2"))))

(event OnLinkDataLoaded(FortPortal)
  (params (LinkData :type Fort%20Game%20Activity%20Object%20Reference))
  ;[0 0.6421 1 0.5] Server Replicated Island Details Loading Logic
  (CallFunction|IslandDataLoaded LinkData))

(event Custom|Attempt%20Matchmaking
  (params (FortPlayerController :type Fort%20Player%20Controller%20Object%20Reference))
  ;[0 0.1399 1 0.5] Matchmaking Logic
  (bind _asfort_player_controller_athena (Utilities|Casting|CastToFortPlayerControllerAthena FortPlayerController))
  (if (CallFunction|ClientOnlyIsPartyLeader _asfort_player_controller_athena)
    (CallFunction|StartMatchmakingandTravel (CallFunction|GetMMSIslandCode))))

(event OnLinkThumbnailUpdated(FortPortal)
  (params (bSuccess :type Boolean))
  ;[0 0.6421 1 0.5] Server Replicated Island Details Loading Logic
  (Networking|CanExecuteCosmeticEvents
    (:True
      (CallFunction|UpdateThumbnailDetails bSuccess (Variables|/Script/FortniteGame.FortPortalComponent|GetThumbnailTexture (Variables|Default|GetFortPortal))))))

(event Custom|Initialize%20Code%20And%20Portal
  ;[0.0199 0.4896 0.1098] Updating HUD When the Device changes.
  (CodeOverride|UpdateLockMode)
  (Networking|CanExecuteCosmeticEvents
    (:True
      (CallFunction|StoreClientPortalComponents)
      (PartyContext|BindEventtoOnLocalPlayerStateChanged (BlueprintContext|GetFortPartyContext) (delegate Custom|Local%20Player%20State%20Changed))
      (bind (_asynctaskproxy _payload) (GameplayEventRouter|StartListeningToEventBackwardCompatible "(TagName=\"EventMessage.Component.Playspace.PlayerExitedSpatialActor.LegacyMinigame\")" (GameplayEventRouter|GetGlobalEventRouter) "/Script/FortniteGame.PlayerExitSpatialActorContextLegacyMinigame" 0 "Down")
        (:then
          (Variables|Default|SetExitedEventHandle _asynctaskproxy))
        (:OnEventReceived
          (CallFunction|OnClientExitVolume (.ExitingPlayerState (Utilities|Struct|BreakPlayerExitSpatialActorContextLegacyMinigame _payload))))))
    (:False
      _))
  (CallFunction|UpdateIslandDetails (Variables|UserOptions|GetIslandCode_Knob)))

(event Custom|Local%20Player%20State%20Changed
  (params (PlayerInfo :type Fort%20Team%20Member%20Info%20Structure%20%28by%20ref%29))
  ;[0.0199 0.4896 0.1098] Updating HUD When the Device changes.
  (CallFunction|UpdateHUD (Variables|IslandDetails|GetCurrent%20Matchmaking%20State)))

(event OnComponentBeginOverlap(HUDVisibilityVolume)
  (params (OverlappedComponent :type Primitive%20Component%20Object%20Reference) (OtherActor :type Actor%20Object%20Reference) (OtherComp :type Primitive%20Component%20Object%20Reference) (OtherBodyIndex :type Integer) (bFromSweep :type Boolean) (SweepResult :type Hit%20Result%20Structure%20%28by%20ref%29))
  ;[1 0.7039 0 0.5] Show or Hide HUD when you're close.
  (CallFunction|SetHUDVisibility OtherActor true))

(event OnComponentEndOverlap(HUDVisibilityVolume)
  (params (OverlappedComponent :type Primitive%20Component%20Object%20Reference) (OtherActor :type Actor%20Object%20Reference) (OtherComp :type Primitive%20Component%20Object%20Reference) (OtherBodyIndex :type Integer))
  ;[1 0.7039 0 0.5] Show or Hide HUD when you're close.
  (CallFunction|SetHUDVisibility OtherActor))

(event Custom|Open%20Code%20Override%20HUD
  (params (Pawn :type Pawn%20Object%20Reference))
  ;[0 0.6904 1] Client Code Override Functions
  (bind _class (Utilities|AsyncLoadClassAsset (Variables|CodeOverride|GetUI_IslandCodeEntrySoftClass))
    (:Completed
      (if (Utilities|IsValidClass _class)
        (bind _asuser_widget (Utilities|Casting|CastToUserWidgetClass _class))
        (bind _returnvalue (UserInterface|CreateWidget _asuser_widget (Fortnite|GetFortPlayerControllerfromActor Pawn)))
        (Init|InitfromObject(Message) _returnvalue self)))))

(event OnPortalLinkCodeLockStatusChanged(FortPortal)
  (params (Status :type EPortalLinkCodeLockStatus%20Enum :default "Unlocked_NotSet"))
  (CodeOverride|UpdateCodeEntryScreen true))

(event Custom|OnClientExitVolume
  (params (Client :type Player%20State%20Object%20Reference))
  ;[1 1 1] When anyone leaves the volume, cancel matchmaking for their whole party
  (if (Variables|Default|GetbLeavingIslandCancelsMatchmaking)
    (bind _asfort_player_controller_athena (Utilities|Casting|CastToFortPlayerControllerAthena (Actor|GetOwner Client)))
    (Class|FortPlayerControllerAthena|CancelMatchmaking _asfort_player_controller_athena "CreativeMemberLeftIsland")
    (GameplayEventRouter|StopListeningtoEvent (Variables|Default|GetExitedEventHandle))))

(event EventDestroyed
  ;[1 1 1] Destroyed - Unbind functions
  (|CallParentFunction)
  (GameplayEventRouter|StopListeningtoEvent (Variables|Default|GetExitedEventHandle)))

(event EventBeginPlay
  (bind _mesh (Variables|Default|GetMesh))
  (|CallParentFunction)
  (if (> (Fortnite|GetNumericCVar "Pilgrim.EnableGreenRoomPortals") 0)
    (CallFunction|InitializeCodeandPortal)
    (else
      (EnabledState|UpdateEnabled false)
      (Development|SetHiddeninGame _mesh true true)
      (Collision|SetCollisionEnabled _mesh "NoCollision"))))

(fn AthenaInteractDetails
  (return false false))

(fn BlueprintCanInteract
  (params (InteractingPawn :type Fort%20Pawn%20Object%20Reference) (InteractionBeingAttempted :type EInteractionBeingAttempted%20Enum :default "FirstInteraction") (InteractionType :type TInteractionType%20Enum :default "IT_NoInteraction"))
  ;[1 1 1] Battlebus interaction currently disabled for Papaya
  (if (CallFunction|CanInteract)
    (Utilities|IsValid (Fortnite|GetFortPlayerControllerfromActor InteractingPawn)
      (:Is%20Valid
        (return (and true (or (Class|FortPortalComponent|IsPartyEligibleforMatchmaking (Variables|Default|GetFortPortal)) true))))
      (:Is%20Not%20Valid
        (return false)))))

(fn BlueprintGetInteractionString
  (params (InteractingPawn :type Fort%20Pawn%20Object%20Reference) (InteractionBeingAttempted :type EInteractionBeingAttempted%20Enum :default "FirstInteraction"))
  (bind _interactingpawn (Variables|GetInteractingPawn))
  (bind _returnvalue (Fortnite|GetFortPlayerControllerfromActor _interactingpawn))
  (bind _client_portal_components (Variables|Default|GetClient%20Portal%20Components))
  (bind (_value _returnvalue_2) (Utilities|Map|Find _client_portal_components _returnvalue))
  (bind _fortportal (Variables|Default|GetFortPortal))
  (if (Variables|CodeOverride|GetLocalPlayerLookingAtScreen)
    (return (Variables|CodeOverride|GetSet%20Code%20Text))
    (else
      (Utilities|IsValid _returnvalue
        (:Is%20Valid
          (if _returnvalue_2
            (if (not (Class|FortPlayerController|IsPartyLeader _returnvalue))
              (return (Variables|Default|GetStart%20Matchmaking%20Text))
              (elif (Variables|Default|GetHasActiveReadyCheck)
                (return (Variables|Default|GetStop%20Matchmaking%20Text))
                (elif (Class|FortControllerComponentPortal|CanInitiatePortalMatchmkaing _value _fortportal)
                  (return (Variables|Default|GetStart%20Matchmaking%20Text))
                  (elif (Class|FortControllerComponentPortal|CanCancelPortalMatchmaking _value)
                    (return (Variables|Default|GetStop%20Matchmaking%20Text))
                    (elif (Class|FortPortalComponent|IsPartyEligibleforMatchmaking _fortportal)
                      (return "INVTEXT(\" \")")
                      (else
                        (return (Variables|Default|GetParty%20Ineligible%20Text))))))))))))))

(fn BlueprintGetInteractionTime
  (params (InteractingPawn :type Fort%20Pawn%20Object%20Reference) (InteractionBeingAttempted :type EInteractionBeingAttempted%20Enum :default "FirstInteraction"))
  (bind _returnvalue (Fortnite|GetFortPlayerControllerfromActor InteractingPawn))
  (Networking|CanExecuteCosmeticEvents
    (:True
      (Utilities|IsValid _returnvalue
        (:Is%20Valid
          (if (Class|FortPlayerController|IsPartyLeader _returnvalue)
            (return (Variables|Default|GetMatchmaking%20Interact%20Time) true)
            (else
              _)))))
    (:False
      _))
  (return 0.0 true))

(fn BlueprintOnLocalInteract
  (params (InteractingPawn :type Fort%20Player%20Pawn%20Object%20Reference) (InteractionBeingAttempted :type EInteractionBeingAttempted%20Enum :default "FirstInteraction"))
  (bind _local_pc (LocalVariables|Getlocal%20PC))
  (bind _client_portal_components (Variables|Default|GetClient%20Portal%20Components))
  (bind (_value _returnvalue) (Utilities|Map|Find _client_portal_components _local_pc))
  (bind _local_pc_2 (LocalVariables|Getlocal%20PC))
  ;[1 1 1] Attempt Matchmake / Cancel Matchmake Attempt
  (Development|PrintString "ONLOCALINTERACT WOOOOO")
  (bind _output_get (LocalVariables|Setlocal%20PC (Fortnite|GetFortPlayerControllerfromActor (Variables|GetInteractingPawn))))
  (Utilities|IsValid _output_get
    (:Is%20Valid
      (if (CallFunction|ClientOnlyIsPartyLeader (LocalVariables|Getlocal%20PC))
        (if _returnvalue
          (if (not (Variables|Default|GetHasActiveReadyCheck))
            (if (Class|FortControllerComponentPortal|CanInitiatePortalMatchmkaing _value (Variables|Default|GetFortPortal))
              (CallFunction|AttemptMatchmaking _local_pc_2)
              (elif (Class|FortControllerComponentPortal|CanCancelPortalMatchmaking _value)
                (CallFunction|CancelMatchmaking _local_pc_2)
                (else
                  (return true))))
            (else
              (CallFunction|CancelReadyCheck)))
          (Variables|Default|SetLastInteractTime (Utilities|Time|GetGameTimeinSeconds)) ; Saving interact time to pause \ninteraction to avoid flipflopping \nbetween states.
          (return true))
        (else
          (FailedAction|UnabletoPerformAction (Variables|GetInteractingPawn) true (GameplayTags|MakeGameplayTagContainerfromTag "(TagName=\"Activate.Failed.Tags.Blocked\")"))
          (return true))))))

(fn CancelReadyCheck
  (bind _readycheckinteractionplayer (Variables|Default|GetReadyCheckInteractionPlayer))
  (Variables|Default|SetHasActiveReadyCheck false)
  (Utilities|IsValid _readycheckinteractionplayer
    (:Is%20Valid
      (Class|FortPlayerControllerAthena|CancelCreativeMatchmakingReadyCheck _readycheckinteractionplayer))))

(fn CanInteract
  (return (and (> (Utilities|Time|GetGameTimeinSeconds) (+ (Variables|Default|GetLastInteractTime) (Variables|Default|GetInteractDelayAfterSuccess))) (Variables|Default|GetbIsEnabled))))

(fn ClientOnlyIsPartyLeader
  (params (FortPC :type Fort%20Player%20Controller%20Object%20Reference))
  (bind _fortpc (Variables|GetFortPC))
  (Networking|CanExecuteCosmeticEvents
    (:True
      (Utilities|IsValid _fortpc
        (:Is%20Valid
          (return (Class|FortPlayerController|IsPartyLeader _fortpc)))
        (:Is%20Not%20Valid
          _)))
    (:False
      _))
  (return false))

(fn ConsoleIsVisible
  (return (Utilities|Enum|NotEqual(Enum) (Variables|/Script/FortniteGame.FortPortalComponent|GetLinkCodeLockMode (Variables|Default|GetFortPortal)) "AlwaysLocked")))

(fn GetDescriptionDetails
  (return ""))

(fn GetHeaderDetails
  (return (select (Variables|CodeOverride|GetLocalPlayerLookingAtScreen) (Variables|CodeOverride|GetLookingAtHeader) (Variables|IslandDetails|GetIsland%20Name)) (Transformation|GetWorldLocation (select (Variables|CodeOverride|GetLocalPlayerLookingAtScreen) 0 (Variables|Default|GetPortal%20Look%20at%20Point))) 0))

(fn GetInteractionDetails
  (params (Controller :type Fort%20Player%20Controller%20Object%20Reference))
  ;[1 1 1] Run through the various matchmaking criteria to figure out what the interaction state is between the player and the portal.
  (Utilities|IsValid Controller
    (:Is%20Valid
      (if (Variables|CodeOverride|GetLocalPlayerLookingAtScreen)
        (return (select (Utilities|Enum|Equal(Enum) (Variables|/Script/FortniteGame.FortPortalComponent|GetLinkCodeLockStatus (Variables|Default|GetFortPortal)) "Locked") "NewEnumerator5" "NewEnumerator4") (Class|FortPortalComponent|GetRemainingLockDuration (Variables|Default|GetFortPortal)))
        (elif (Class|FortPortalComponent|IsPartyEligibleforMatchmaking (Variables|Default|GetFortPortal))
          (return (select (CallFunction|ClientOnlyIsPartyLeader Controller) "NewEnumerator0" "NewEnumerator2") 0)
          (else
            (return (select (Utilities|Enum|Equal(Enum) (Variables|IslandDetails|GetCurrent%20Matchmaking%20State) "NewEnumerator3") "NewEnumerator3" "NewEnumerator0") 0)))))))

(fn GetIslandPortalDetails
  (return (select (Variables|CodeOverride|GetLocalPlayerLookingAtScreen) "NewEnumerator0" (Variables|IslandDetails|GetCurrent%20Matchmaking%20State)) (Variables|IslandDetails|GetCreator%20Name) (Variables|IslandDetails|GetDescription%20Text) (Utilities|Text|ToText(String) (Variables|Default|GetCurrent%20Island%20Code)) (Variables|IslandDetails|GetPopMin) (Variables|IslandDetails|GetPopMax)))

(fn GetBoundsComponents
  (params (SelfActor :type Actor%20Object%20Reference))
  (return (Utilities|Array|MakeArray (Variables|Default|GetSphere))))

(fn GetMMSIslandCode
  (if (CallFunction|ShouldUseBlueprintMMSCode)
    (return (Variables|UserOptions|GetIslandCode_Knob))
    (else
      (bind _returnvalue (CallFunction|GetMMSCodeforLocalPlayerParty))
      (return _returnvalue))))

(fn IslandDataLoaded
  (params (LinkData :type Fort%20Game%20Activity%20Object%20Reference%20%28by%20ref%29))
  (bind _returnvalue (Class|FortGameActivity|GetMatchmakingSettingsV1 LinkData))
  (bind (_minimumnumberofplayers _maximumnumberofplayers) (Utilities|Struct|BreakCreativeIslandMatchmakingSettings _returnvalue))
  (bind _linkid (Variables|/Script/FortniteGame.FortGameActivity|GetLinkId LinkData))
  (bind _mnemonic (Utilities|Struct|BreakOnlineLinkId _linkid))
  (bind _linkdata (Variables|GetLinkData))
  (if (> (Utilities|String|Len _mnemonic) 0)
    (Variables|Default|SetCurrent%20Island%20Code _mnemonic)
    (Networking|CanExecuteCosmeticEvents
      (:True
        (Variables|IslandDetails|SetCreator%20Name (Utilities|Text|ToText(String) (Variables|/Script/FortniteGame.FortGameActivity|GetCreator LinkData)))
        (Variables|IslandDetails|SetIsland%20Name (Utilities|Text|ToText(String) (Variables|/Script/FortniteGame.FortGameActivity|GetTitle _linkdata)))
        (Variables|IslandDetails|SetDescription%20Text (Utilities|Text|ToText(String) (Variables|/Script/FortniteGame.FortGameActivity|GetTagline _linkdata)))
        (Variables|IslandDetails|SetPopMin _minimumnumberofplayers)
        (Variables|IslandDetails|SetPopMax _maximumnumberofplayers)
        (CallFunction|InitializeCodeandPortal)
        (CallFunction|UpdateHUD "NewEnumerator3")))))

(fn IslandDataNoSuccess
  (params (Matchmaking%20State :type ECreativeMatchmakingPortalState%20Enum :default "NewEnumerator0"))
  (Networking|CanExecuteCosmeticEvents
    (:True
      (Variables|IslandDetails|SetIsland%20Name "NSLOCTEXT(\"[6B8C92FF7D7F651A42F42D0ABAFF3A1C]\", \"04C702C64BD04EA8DCDCED8FA944B0C3\", \"Battle Royale\")")
      (Variables|IslandDetails|SetThumbnailTexture "")
      (CallFunction|UpdateHUD Matchmaking%20State)
      (CallFunction|UpdatePortalVisuals))))

(fn MatchmakingPortalHasSecondInteraction
  (params (InteractingPawn :type Fort%20Player%20Pawn%20Object%20Reference))
  (return false))

(fn OnUserGeneratedContentRestrictionChanged
  (params (bEnabled :type Boolean))
  (if bEnabled
    (Actor|DestroyActor)))

(fn PortalHUDIsVisible
  (params (bVisibility :type Boolean))
  (return bVisibility))

(fn PortalThumbnailIsVisible
  (return true))

(fn ReadyCheckCompleted
  (Variables|Default|SetHasActiveReadyCheck false)
  (bind _readycheckinteractionplayer (Variables|Default|GetReadyCheckInteractionPlayer)
    (:then
      (EventDispatchers|UnbindEventfromInGameMatchmakingReadyCheckComplete _readycheckinteractionplayer (EventDispatchers|CreateEvent :!event ReadyCheckCompleted))
      (bind _returnvalue (Class|FortCreativeDeviceProp|CreateLinkId (CallFunction|GetMMSIslandCode)))
      (Class|FortPlayerControllerAthena|StartMatchmakingbyLinkCode _readycheckinteractionplayer _returnvalue false))
    (:else)))

(fn ScaleThumbnailbyTextureSize
  (params (ScaleValue :type Vector%202D%20Structure))
  (bind (_scalevalue_x _scalevalue_y) (Math|Vector2D|BreakVector2D (Variables|GetScaleValue_X)))
  (return (* (Utilities|Select (Math|Vector|MakeVector (/ (Variables|Default|GetThumbnail%20Size) (/ (Math|Conversions|ToFloat(Integer) (Math|Float|Truncate _scalevalue_x)) (Math|Conversions|ToFloat(Integer) (Math|Float|Truncate _scalevalue_y)))) (Variables|Default|GetThumbnail%20Size) 1.0) (Math|Vector|MakeVector (Variables|Default|GetThumbnail%20Size) (/ (Variables|Default|GetThumbnail%20Size) (/ (Math|Conversions|ToFloat(Integer) (Math|Float|Truncate _scalevalue_x)) (Math|Conversions|ToFloat(Integer) (Math|Float|Truncate _scalevalue_y)))) 1.0) (> (Math|Float|Truncate _scalevalue_x) (Math|Float|Truncate _scalevalue_y))) (Transformation|GetActorScale3D))))

(fn SendIslandLink
  (params (Requesting%20Player%20Controller :type Fort%20Player%20Controller%20Object%20Reference) (Link :type Fort%20Creative%20Island%20Link%20Object%20Reference))
  (Utilities|IsValid Link
    (:Is%20Valid
      (CallFunction|UpdateIslandDetails (Utilities|Struct|BreakOnlineLinkId (Variables|/Script/FortniteUI.FortCreativeIslandLink|GetLinkId Link)) true Requesting%20Player%20Controller)
      (Variables|Default|SetLastInteractTime (Utilities|Time|GetGameTimeinSeconds)) ; Saving interact time. Used to disable interaction\nfor a short time.
      (return true))
    (:Is%20Not%20Valid
      (return false))))

(fn SetHUDVisibility
  (params (Overlapping%20Actor :type Actor%20Object%20Reference) (bVisibility :type Boolean))
  ;[1 0.7441 0] Shows or Hides the Widget if you get close enough  / far enough away from it
  (Networking|CanExecuteCosmeticEvents
    (:True
      (Variables|Default|GetHUD%20Widget%20Component
        (:then
          (Class|CreativeHeaderDescriptionHUDComponent|ShowWidget (Variables|Default|GetHUD%20Widget%20Component) Overlapping%20Actor (CallFunction|PortalHUDIsVisible (Variables|GetbVisibility))))
        (:else)))))

(fn SetupWaitForReadyCheck
  (params (FortPlayer :type Fort%20Player%20Controller%20Athena%20Object%20Reference))
  (bind _outputdelegate (EventDispatchers|CreateEvent :!event ReadyCheckCompleted))
  (bind _fortplayer (Variables|GetFortPlayer))
  (Utilities|IsValid FortPlayer
    (:Is%20Valid
      (Variables|Default|SetReadyCheckInteractionPlayer FortPlayer)
      (Variables|Default|SetHasActiveReadyCheck true)
      (EventDispatchers|UnbindEventfromInGameMatchmakingReadyCheckComplete _fortplayer _outputdelegate)
      (EventDispatchers|BindEventtoInGameMatchmakingReadyCheckComplete _fortplayer _outputdelegate))))

(fn StoreClientPortalComponents
  (for _array_element (Fortnite|GetLocalFortPlayerControllers true)
    (Utilities|IsValid (Actor|GetComponentbyClass _array_element "/Script/FortniteGame.FortControllerComponent_Portal")
      (:Is%20Valid
        (Utilities|Map|Add (Variables|Default|GetClient%20Portal%20Components) _array_element (Actor|GetComponentbyClass _array_element "/Script/FortniteGame.FortControllerComponent_Portal"))))))

(fn UpdateCodeEntryScreen
  (params (Force%20Screen%20Update :type Boolean))
  (LocalVariables|SetMID (Material|GetMID 0))
  (if (not (Math|Conversions|ToBoolean(Integer) (Utilities|Array|Length (Variables|CodeOverride|GetLocal%20Pawns%20in%20Range))))
    (if (or Force%20Screen%20Update (Math|Boolean|NotEqual(Boolean) (LocalVariables|GetLookingAtScreen) (Variables|CodeOverride|GetLocalPlayerLookingAtScreen)))
      (Variables|CodeOverride|SetLocalPlayerLookingAtScreen (LocalVariables|GetLookingAtScreen))
      (bind _hud_widget_component (Variables|Default|GetHUD%20Widget%20Component)
        (:then
          (Class|CreativeHeaderDescriptionHUDComponent|SpawnOrUpdateHUD _hud_widget_component "NewEnumerator1"))
        (:else)))))

(fn UpdateEnabled
  (params (Enabled :type Boolean))
  (Variables|Default|SetbIsEnabled Enabled)
  (Networking|CanExecuteCosmeticEvents
    (:True
      (Collision|SetCollisionEnabled (Variables|Default|GetHUD%20Visibility%20Volume) (select Enabled "QueryOnly" "NoCollision")) ; Disabling or enabling collision on the visibility volume\nwill trigger the overlap events to enable/disable the HUD.
      (CallFunction|UpdatePortalVisuals)
      (EnabledState|UpdatePoweredVisuals)
      (if (not (Variables|GetEnabled))
        (bind _hud_widget_component (Variables|Default|GetHUD%20Widget%20Component)
          (:then
            (Class|CreativeHeaderDescriptionHUDComponent|ForceShowWidget _hud_widget_component false))
          (:else))))))

(fn UpdateHUD
  (params (Matchmaking%20State :type ECreativeMatchmakingPortalState%20Enum :default "NewEnumerator0"))
  (Variables|IslandDetails|SetCurrent%20Matchmaking%20State Matchmaking%20State)
  (Networking|CanExecuteCosmeticEvents
    (:True
      (bind _hud_widget_component (Variables|Default|GetHUD%20Widget%20Component)
        (:then
          (Class|CreativeHeaderDescriptionHUDComponent|SpawnOrUpdateHUD _hud_widget_component "NewEnumerator1"))
        (:else)))
    (:False
      (bind _hud_visibility_volume (Variables|Default|GetHUD%20Visibility%20Volume)
        (:then
          (Components|DestroyComponent _hud_visibility_volume) ; We don't have any reason to show a widget, so no need to have a visibility checking volume.\nNo need to destroy the widget also, it will do that by itself.
        )
        (:else)))))

(fn UpdateIslandDetails
  (params (Code :type String) (Logic%20on%20Client :type Boolean) (Client%20Player%20Controller :type Fort%20Player%20Controller%20Object%20Reference))
  (bind _client_portal_components (Variables|Default|GetClient%20Portal%20Components))
  (Utilities|BasicCodeValidation Code
    (:Valid%20Code
      (Networking|SwitchHasAuthority
        (:Authority
          (if (not (Variables|GetLogic%20on%20Client))
            (bind _returnvalue (Class|FortCreativeDeviceProp|CreateLinkId (Variables|GetCode)))
            (Class|FortPortalComponent|ServerSetLinkCode (Variables|Default|GetFortPortal) _returnvalue false false false)))
        (:Remote
          (CallFunction|IslandDataNoSuccess "NewEnumerator1")
          (if (Variables|GetLogic%20on%20Client)
            (bind _client_player_controller (Variables|GetClient%20Player%20Controller)
              (:then
                (if (.ReturnValue (Utilities|Map|Find _client_portal_components _client_player_controller))
                  (bind _returnvalue_2 (Class|FortCreativeDeviceProp|CreateLinkId (Variables|GetCode)))
                  (Class|FortControllerComponentPortal|ServerSetPortalLinkCode (Utilities|Map|Find _client_portal_components _client_player_controller) (Variables|Default|GetFortPortal) _returnvalue_2)))
              (:else))))))) ; Runs Validation Logic prior to submission \nagainst known bad options (defualt, string length too short).\nTriggers appropriate error messages.
)

(fn UpdateLockMode
  (Variables|/Script/FortniteGame.FortPortalComponent|SetLinkCodeLockMode (Variables|UserOptions|GetCanOverrideCode) (Variables|Default|GetFortPortal)))

(fn UpdatePortalVisuals
  (Development|SetHiddeninGame (Variables|Default|GetMesh) (not (CallFunction|PortalThumbnailIsVisible)) false))

(fn UpdatePoweredVisuals)

(fn UpdateThumbnailDetails
  (params (Success :type Boolean) (Texture :type Texture%202DDynamic%20Object%20Reference))
  (if Success
    (Utilities|IsValid Texture
      (:Is%20Valid
        (Variables|IslandDetails|SetThumbnailTexture Texture)
        (Networking|CanExecuteCosmeticEvents
          (:True
            (Variables|IslandDetails|GetThumbnailTexture
              (:then
                (CallFunction|UpdatePortalVisuals))
              (:else
                _)))))
      (:Is%20Not%20Valid
        _))
    (CallFunction|UpdatePortalVisuals)
    (else
      _)))

(fn ConstructionScript
  (|CallParentFunction)
  (Networking|SwitchHasAuthority
    (:Authority
      (Collision|SetCollisionEnabled (Variables|Default|GetHUD%20Visibility%20Volume) "NoCollision"))))
