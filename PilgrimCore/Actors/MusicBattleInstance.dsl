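; Music battle instance actor for Pilgrim, running a song session's note tracks, gameplay UI and audio hooks.
; Clients create tracks and games for local players, padded by debug tracks up to four, then push a touch, single or multiplayer UI scene.
; Game end is a handshake: clients report song finished, replicated bShouldEnd triggers delayed teardown, then clients notify the server.
; Also triggers per-instrument disrupt FX and end-of-song crowd cheers.
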(event Custom|TriggerDisruptAudio
  ;[1 0 0.3462] Start + Stop Sonic Attack Disrupt Momentary Stem Audio FX
  (switch Utilities|FlowControl|Switch|SwitchonEPilgrimTrackType (Class|PilgrimCorePlayerStateComponent|GetInstrumentChoice (Actor|GetComponentbyClass "/Script/PilgrimCoreRuntime.PilgrimCorePlayerStateComponent"))
    (:TrackGuitar
      (CallFunction|ExecuteTriggerForGuitar))
    (:TrackBass
      (CallFunction|ExecuteTriggerForBass))
    (:TrackVocals
      (CallFunction|ExecuteTriggerForVocals))
    (:TrackDrum
      (CallFunction|ExecuteTriggerForDrums))
    (:TrackPlasticGuitar
      (CallFunction|ExecuteTriggerForGuitar))
    (:TrackPlasticBass
      (CallFunction|ExecuteTriggerForBass))
    (:TrackPlasticDrum
      (CallFunction|ExecuteTriggerForDrums))
    (:TrackProVocals
      (CallFunction|ExecuteTriggerForVocals))
    (:TrackProCymbals
      (CallFunction|ExecuteTriggerForDrums))))

(event Custom|StopDisruptAudio
  ;[1 0 0.3462] Start + Stop Sonic Attack Disrupt Momentary Stem Audio FX
  (Audio|Parameter|ExecuteTriggerParameter (CallFunction|GetPlayerAudioComponent) "DisruptStop"))

(event Custom|On%20Song%20Start%20Audio%20Events)

(event Custom|On%20Song%20End%20Audio%20Events
  ;[0.9313 0 1] AUDIO:  Play one-off cheers on song start & end.
  ; Song End no longer is cleanerup by BP unloading
  (Audio|PlaySound2D "/SparksMaps/Audio/Crowd/Festival_Crowd_Arena_SongEnd.Festival_Crowd_Arena_SongEnd"))

(event EventBeginPlay
  ;[1 1 1] Allows FN BR Audio Reactivity to work.  Metasound file handles the rest - sending to FN audio submix
  (AudioAnalysis|RegisterDefaultSubmix (WorldSubsystems|GetFortAudioAnalysisSubsystem)))

(event EventOnClientReceivedSongToPlay
  (params (Song :type Pilgrim%20Song%20Metadata%20Object%20Reference))
  ;[1 1 1] Comment
  (Networking|SwitchHasAuthority
    (:Remote
      (CallFunction|MakePilgrimGamesforLocalPlayers))))

(event EventOnStartedPlayingSong
  (params (Song :type Pilgrim%20Song%20Metadata%20Object%20Reference))
  ;[1 1 1] Called after we Play the Song. Now all song loading and playing is being done natively
  (CallFunction|SetupTracksBasedonConfigurableOptions)
  (Utilities|FlowControl|Sequence
    (:then_0
      ;;[0.481 0.5544 1] Trigger audio events for song start
      (CallFunction|OnSongStartAudioEvents))
    (:then_1
      ;;[0 0.7292 0.3064] Tell anyone who cares about the MBI starting
      (CallFunction|BroadcastMBIEventStarted))
    (:then_2
      ;;[1 1 1] We only want to tick on the client if a player is in the battle. However, for spectators, we probably want one MBI ticking...
      (if (CallFunction|IsLocalPlayerinBattleParticipants)
        (Actor|Tick|SetActorTickEnabled true)))))

(event EventUpdateScore
  (if (CallFunction|IsLocalPlayerinBattleParticipants)
    (bind _returnvalue (GameplayEventRouter|GetEventRouter "/Script/FortniteGame.FortPlayspace" self))
    (GameplayEventRouter|BroadcastEvent _returnvalue (Variables|PilgrimMusicBattleInstance|GetMusicBattleData) 0)))

(event EventServerUserFinishedDisconnecting
  ;[1 1 1] Wait for everyone to shut down before destroying the actor.
  (Networking|SwitchHasAuthority
    (:Authority
      (Actor|DestroyActor))))

(event EventEndPlay
  (params (EndPlayReason :type EEndPlayReason%20Enum :default "Destroyed"))
  (Default|ClientFinishCleanUpPilgrimGame))

(event Custom|ClientDoCleanup
  ;[0.4478 1 0.9901] Game End Behaviors
  (Networking|SwitchHasAuthority
    (:Remote
      ;;[1 1 1] Client: Start tearing down, wait, clean up completely, then notify server.
      (if (not (Variables|Default|GetbHasStartedCleanup))
        (Variables|Default|SetbHasStartedCleanup true)
        (Default|ClientStartCleanUpPilgrimGame)
        ;;;[0.481 0.5544 1] Trigger audio events for song end
        (CallFunction|OnSongEndAudioEvents)
        ;;;[1 0.329 0.6395] Wait for track to dissolve/clear
        (Utilities|FlowControl|Delay (Variables|Default|GetTearDownDuration))
        (Default|ClientFinishCleanUpPilgrimGame)
        (Default|LocalClientSendGameEndToServer)))))

(event Custom|OnRep_bShouldEnd
  ;[0.4478 1 0.9901] Game End Behaviors
  (if (Variables|Default|GetbShouldEnd)
    (CallFunction|ClientDoCleanup)))

(event Custom|OnLocalSongEnded
  (params (PilgrimGame :type Pilgrim%20Game%20Object%20Reference))
  ;[0.4478 1 0.9901] Game End Behaviors
  (Default|LocalClientSendStateUpdateToServer "SongFinished"))

(event EventServerAllPlayersHaveFinishedSong
  ;[0.4478 1 0.9901] Game End Behaviors
  ;;[1 1 1] When all players have finished the song, start the teardown.
  (Variables|Default|SetbShouldEnd true))

(event EventOnTrackHiddenChanged
  (params (bHideTrack :type Boolean))
  (for _array_element (Variables|Music|GetAllTracks)
    (Transformation|SetActorRelativeScale3D _array_element (select bHideTrack "0, 0, 0" "1.000000,1.000000,1.000000"))))

(event EventUpdateVisibilityOfUI
  (bind _returnvalue (WorldSubsystems|GetDynamicUIManager))
  (bind _uiscene (Variables|Default|GetUIScene)
    (:then
      (bind _returnvalue_2 (Development|GetConsoleVariableBoolValue "Pilgrim.ShowAllUI"))
      (if _returnvalue_2
        (DynamicUI|AddScenetoFirstLocalPlayer _returnvalue _uiscene)
        (else
          (DynamicUI|RemoveScenefromFirstLocalPlayer _returnvalue _uiscene))))
    (:else)))

(fn ClientFinishCleanUpPilgrimGame
  (if (not (Variables|Default|GetbHasFinishedCleanup))
    (Utilities|FlowControl|Sequence
      (:then_0
        (bind _returnvalue (GameplayEventRouter|GetEventRouter "/Script/FortniteGame.FortPlayspace" self))
        (GameplayEventRouter|BroadcastEvent _returnvalue (Utilities|Struct|MakePilgrimEventPilgrimGameTrackFadeEnded) 0)
        (Utilities|FlowControl|Sequence
          (:then_0
            (for _array_element (Variables|Music|GetAllTracks)
              (Rendering|SetActorHiddenInGame _array_element true)
              (Class|PilgrimTrack|HideAllItems _array_element)
              (Actor|DestroyActor _array_element)))
          (:then_1
            (CallFunction|ClientStopSongbyServer))))
      (:then_1
        (if (CallFunction|IsLocalPlayerinBattleParticipants)
          (if (CallFunction|CanSafelyClearUIScene)
            (bind _uiscene (Variables|Default|GetUIScene)
              (:then
                (DynamicUI|RemoveScenefromFirstLocalPlayer (WorldSubsystems|GetDynamicUIManager) _uiscene)
                (CallFunction|OnClearedUIScene))
              (:else))
            (else
              (CallFunction|LogUIEvent "Not clearing UI Scene because it was deemed unsafe.")))))
      (:then_2
        (Variables|Default|SetbHasFinishedCleanup true)))))

(fn ClientStartCleanUpPilgrimGame
  (Utilities|FlowControl|Sequence
    (:then_0
      (for _array_element (Variables|Music|GetAllTracks)
        (Class|PilgrimTrack|OnGameEnded _array_element)))
    (:then_1
      (for _array_element_2 (CallFunction|GetLocalPilgrimGames)
        (Class|PilgrimGame|EndGame _array_element_2 true 2.5)))
    (:then_2
      ;[0 0.7292 0.3064] Fire a Gameplay event about the game ending
      (CallFunction|BroadcastMBIEventFinished))))

(fn CreatePilgrimBattleForLocalPlayer
  (params (InLocalPC :type Fort%20Player%20Controller%20Object%20Reference) (IsPrimaryPlayer :type Boolean :default true) (CurrentTrackIndex :type Integer) (TotalTracksToCreate :type Integer :default 1))
  (bind _returnvalue (CallFunction|GetTrackandGameManager))
  (Networking|SwitchHasAuthority
    (:Remote
      (if (CallFunction|IsLocalPlayerinBattleParticipants)
        (Utilities|FlowControl|Sequence
          (:then_0
            ; If we have a local player, get SongDifficulty/TrackType from them
            (bind _inlocalpc (Variables|GetInLocalPC)
              (:then
                (LocalVariables|SetDifficulty (CallFunction|GetDifficultyandTrackTypeforPlayer (Variables|/Script/Engine.Controller|GetPlayerState _inlocalpc)))
                (LocalVariables|SetTrackType (.OutSongTrackType (CallFunction|GetDifficultyandTrackTypeforPlayer (Variables|/Script/Engine.Controller|GetPlayerState _inlocalpc)))))
              (:else
                ;;[0.15 0.03281 0.03281 0.704] For "ownerless" tracks (debug utility) prefer expert difficulty, and predetermined instrument choice.
                (LocalVariables|SetDifficulty "DifficultyExpert")
                (LocalVariables|SetTrackType (Utilities|Select :Index (Variables|GetCurrentTrackIndex) :Option%200 "TrackGuitar" :Option%201 "TrackBass" :Option%202 "TrackVocals" :Option%203 "TrackDrum")))))
          (:then_1
            ; Don't make a track for ProVocals
            (if (Utilities|Enum|NotEqual(Enum) (LocalVariables|GetTrackType) "TrackProVocals")
              (Utilities|IsValid _returnvalue
                (:Is%20Valid
                  (bind _returnvalue_2 (Class|PilgrimTrackandGameManagerComponent|CreateTrackandGameforPlayer _returnvalue (Utilities|Struct|MakePilgrimCreateTrackAndGemsParams self (CallFunction|GetConfig) (Variables|GetInLocalPC) (LocalVariables|GetDifficulty) (LocalVariables|GetTrackType) (Variables|GetCurrentTrackIndex) (Variables|GetTotalTracksToCreate))))
                  (bind _asbp_pilgrim_track (Utilities|Casting|CastToBP_PilgrimTrack (Utilities|Struct|BreakPilgrimTrackandGame _returnvalue_2)))
                  (Utilities|Array|Add (Variables|Music|GetAllTracks) _asbp_pilgrim_track)
                  (Utilities|Array|Add (Variables|PilgrimMusicBattleInstance|GetClientPilgrimGames) (.PilgrimGame (Utilities|Struct|BreakPilgrimTrackandGame _returnvalue_2)))
                  (EventDispatchers|BindEventtoOnSongEnded (.PilgrimGame (Utilities|Struct|BreakPilgrimTrackandGame _returnvalue_2)) (EventDispatchers|CreateEvent :!event OnLocalSongEnded))
                  (Utilities|IsValid (Variables|Music|GetPrimaryTrack)
                    (:Is%20Not%20Valid
                      (Variables|Music|SetPrimaryTrack (Variables|Music|GetMusicTrack))))))))
          (:then_2
            (if IsPrimaryPlayer
              ;[0.1653 1 0.5948] Push Gameplay UI
              (bind _desired_ui_scene (CallFunction|GetDesiredDynamicUIScene (> TotalTracksToCreate 1)))
              (Variables|Default|SetUIScene _desired_ui_scene)
              (bind _returnvalue_3 (Development|GetConsoleVariableBoolValue "Pilgrim.ShowAllUI"))
              (if _returnvalue_3
                (Utilities|FlowControl|Sequence
                  (:then_0
                    (if (DynamicUI|IsSceneActiveonFirstLocalPlayer (WorldSubsystems|GetDynamicUIManager) (Variables|Default|GetUIScene))
                      (CallFunction|LogUISceneAlreadyPresent)
                      (DynamicUI|RemoveScenefromFirstLocalPlayer (WorldSubsystems|GetDynamicUIManager) (Variables|Default|GetUIScene))))
                  (:then_1
                    (DynamicUI|AddScenetoFirstLocalPlayer (WorldSubsystems|GetDynamicUIManager) (Variables|Default|GetUIScene))
                    (CallFunction|OnPushedUIScene)))))))))))

(fn DoMultiplayerDamage
  (params (DamagingPlayerId :type Integer) (DamagedPlayerId :type Integer) (GameUpdateThatDidDamage :type Pilgrim%20Game%20Progress%20Structure%20%28by%20ref%29)))

(fn ExecuteTriggerforBass
  (Audio|Parameter|ExecuteTriggerParameter (CallFunction|GetPlayerAudioComponent) "DisruptFXBass"))

(fn ExecuteTriggerforDrums
  (Audio|Parameter|ExecuteTriggerParameter (CallFunction|GetPlayerAudioComponent) "DisruptFXDrums"))

(fn ExecuteTriggerforGuitar
  (Audio|Parameter|ExecuteTriggerParameter (CallFunction|GetPlayerAudioComponent) "DisruptFXGuitar"))

(fn ExecuteTriggerforVocals
  (Audio|Parameter|ExecuteTriggerParameter (CallFunction|GetPlayerAudioComponent) "DisruptFXVocals"))

(fn GetDesiredDynamicUIScene
  (params (Is%20Playing%20Local%20Multiplayer :type Boolean))
  (bind _uistateconfig (Variables|PilgrimMusicBattleInstance|GetUIStateConfig)
    (:then
      (if (TouchPlatform|SupportsTouchInputDuringGameplay)
        (return (Variables|/Script/PilgrimCoreRuntime.PilgrimMusicBattleInstanceUIStateConfig|GetTouchUIScene _uistateconfig))
        (else
          (return (select (Variables|GetIs%20Playing%20Local%20Multiplayer) (Variables|/Script/PilgrimCoreRuntime.PilgrimMusicBattleInstanceUIStateConfig|GetLocalMultiplayerUIScene _uistateconfig) (Variables|/Script/PilgrimCoreRuntime.PilgrimMusicBattleInstanceUIStateConfig|GetLocalSingleplayerUIScene _uistateconfig))))))
    (:else
      (return 0))))

(fn GetWorldLocationfromNormalizedScreenCoords
  (params (Track%20Horizontal :type Float%20%28double-precision%29) (Track%20Vertical :type Float%20%28double-precision%29) (Track%20Depth :type Float%20%28double-precision%29) (Player%20Controller :type Fort%20Player%20Controller%20Object%20Reference))
  (bind (_sizex _sizey) (HUD|GetViewportSize Player%20Controller))
  (bind (_worldposition _worlddirection) (Camera|DeprojectScreentoWorld Player%20Controller (Math|Vector2D|MakeVector2D (* _sizex Track%20Horizontal) (* _sizey Track%20Vertical))))
  (return (+ _worldposition (* _worlddirection Track%20Depth))))

(fn GetLocalPilgrimGame
  (return (Variables|Default|GetPilgrimGameInstance)))

(fn LocalClientSendGameEndtoServer
  (if (and (CallFunction|IsLocalPlayerinBattleParticipants) (not (Variables|Default|GetbLocalClientHasSignaledToEnd)))
    (Variables|Default|SetbLocalClientHasSignaledToEnd true)
    (Default|LocalClientSendStateUpdateToServer "TearDownFinished")))

(fn LocalClientSendStateUpdatetoServer
  (params (MusicBattleUpdate :type EMusicBattleClientUpdate%20Enum :default "SongFinished"))
  (if (CallFunction|IsLocalPlayerinBattleParticipants)
    (bind _returnvalue (CallFunction|GetLocalPlayerControllers))
    (for _array_element _returnvalue
      (Utilities|IsValid (Actor|GetComponentbyClass _array_element "/Script/PilgrimCoreRuntime.PilgrimPlayerControllerComponent")
        (:Is%20Valid
          (Class|PilgrimPlayerControllerComponent|SendUpdatetoMusicBattle (Actor|GetComponentbyClass _array_element "/Script/PilgrimCoreRuntime.PilgrimPlayerControllerComponent") MusicBattleUpdate))))))

(fn MakePilgrimGamesForLocalPlayers
  (Utilities|FlowControl|Sequence
    (:then_0
      ; Add a means for us to fake extra tracks via cvar. Always make between 1 and 4 tracks.
      (bind _returnvalue (CallFunction|GetLocalPlayerControllers))
      (LocalVariables|SetLocalControllers _returnvalue)
      (LocalVariables|SetNumTracksToCreate (Math|Integer|Clamp(Integer) (Math|Integer|Max(Integer) (Utilities|Array|Length (LocalVariables|GetLocalControllers)) (CallFunction|DebugGetTestMultiTrackDisplay)) 1 4)))
    (:then_1
      ; TODO: Make function on MBI that returns local players with the primary player first.
      (for _array_element (LocalVariables|GetLocalControllers)
        (CallFunction|CreatePilgrimBattleforLocalPlayer _array_element (CallFunction|IsPrimaryLocalPlayer _array_element) (LocalVariables|GetNumTracksCreated) (LocalVariables|GetNumTracksToCreate))
        (Math|Integer|IncrementInt (LocalVariables|GetNumTracksCreated))))
    (:then_2
      (while (< (LocalVariables|GetNumTracksCreated) (LocalVariables|GetNumTracksToCreate))
        (CallFunction|CreatePilgrimBattleforLocalPlayer 0 false (LocalVariables|GetNumTracksCreated) (LocalVariables|GetNumTracksToCreate))
        (Math|Integer|IncrementInt (LocalVariables|GetNumTracksCreated))))
    (:then_3
      ;[0.15 0 0.005254 0.5] If we made more than one track
      (if (> (LocalVariables|GetNumTracksCreated) 1)
        _))))

(fn OnClientWidgetsCreated
  (params (PlayerPawn :type Fort%20Player%20Pawn%20Object%20Reference)))

(fn SetCurrentMusicBattleOptions
  (params (CurrentMusicBattleOptions :type Music%20Battle%20Options%20Structure))
  (Variables|Default|SetCurrentMusicBattleOptions CurrentMusicBattleOptions)
  (for _array_element (Variables|Music|GetAllTracks)
    (CallFunction|UpdateTrackForCurrentMusicBattleOptions _array_element)))

(fn SetupTracksBasedOnConfigurableOptions
  (Default|SetCurrentMusicBattleOptions (Variables|Default|GetConstInitialMusicBattleOptions))
  (for _array_element (Variables|Music|GetAllTracks)
    (Rendering|SetActorHiddenInGame _array_element false)
    (if (CallFunction|DebugIsTrackHidden)
      (Transformation|SetActorRelativeScale3D _array_element "0, 0, 0"))))

(fn UpdateTrackforCurrentMusicBattleOptions
  (params (InTrack :type BP%20Pilgrim%20Track%20Object%20Reference))
  (bind _currentmusicbattleoptions (Variables|Default|GetCurrentMusicBattleOptions))
  (bind (_trackvisible_4_c10ed51e42d29f6e2874c6bd9161a585 _trackangle_8_92e985ab40016303f33eec8e00c8b09c _trackyaw_28_7c1e2f114bb73910dddf3e90a2bb91c3 _trackhorizontal_19_23c1ab764812c1aa01711193b08ffd81 _trackvertical_21_ce07e93b46dd488304ccc1829aa9044f _trackdepth_26_c8474826401cf785c44cad83d698f7a6 _trackcollision_23_a9d8249543dd587c9bc090b326c9bea3 _showbeatmarkers_16_ba8a5f3d4d1004d1edf89db8b98d54a1) (Utilities|Struct|BreakMusicBattleOptions _currentmusicbattleoptions))
  (bind (_sizex _sizey) (HUD|GetViewportSize (LocalVariables|GetLocalPlayerController)))
  (Utilities|IsValid InTrack
    (:Is%20Valid
      (Utilities|FlowControl|Sequence
        (:then_0
          ;[1 1 1] Set local variables
          (LocalVariables|SetTrackAngle _trackangle_8_92e985ab40016303f33eec8e00c8b09c)
          (LocalVariables|SetTrackYaw _trackyaw_28_7c1e2f114bb73910dddf3e90a2bb91c3)
          (LocalVariables|SetTrackHorizontal _trackhorizontal_19_23c1ab764812c1aa01711193b08ffd81)
          (LocalVariables|SetTrackVertical (- 1.0 _trackvertical_21_ce07e93b46dd488304ccc1829aa9044f))
          (LocalVariables|SetTrackDepth _trackdepth_26_c8474826401cf785c44cad83d698f7a6)
          (LocalVariables|SetTrackCollision _trackcollision_23_a9d8249543dd587c9bc090b326c9bea3)
          (LocalVariables|SetShowBeatMarkers _showbeatmarkers_16_ba8a5f3d4d1004d1edf89db8b98d54a1)
          ;; LocalMP: This *Looks* bad but it's just for listening to some debug widget stuff that will be owned by the primary player anyway.
          (LocalVariables|SetLocalPlayerController (Fortnite|GetFirstLocalFortPlayerController true))
          (LocalVariables|SetLocalPlayerPawn (Game|Player|GetPlayerPawn (LocalVariables|GetLocalPlayerController))))
        (:then_1
          ;[1 1 1] Set Track Collision
          (Collision|SetActorEnableCollisionwithChildren (Variables|GetInTrack) (LocalVariables|GetTrackCollision)))
        (:then_2
          ;[1 1 1] Let the track itself respond to option changes.
          (Default|OnPilgrimBattleOptionsChanged (Variables|GetInTrack) (Variables|Default|GetCurrentMusicBattleOptions)))
        (:then_3
          ;[1 1 1] Debug Screen Location to World Coordinates
          (bind _worldposition (CallFunction|GetWorldLocationFromNormalizedScreenCoords (LocalVariables|GetTrackHorizontal) (LocalVariables|GetTrackVertical) (LocalVariables|GetTrackDepth) (LocalVariables|GetLocalPlayerController)))
          (Rendering|Debug|DrawDebugArrow _worldposition (Camera|DeprojectScreentoWorld (LocalVariables|GetLocalPlayerController) (Math|Vector2D|MakeVector2D (* _sizex 0.5) (* _sizey 0.5))) 0.2 "(R=0.000000,G=0.059466,B=0.612847,A=1.000000)" 0.3 0.2))))))

(fn ConstructionScript)
